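# Niagara idle cosmetic effect for the DimeBlanketGrace character, with emitters including Hearts and PixieDust sprites that follow the body.
# Particle partitions select skeletal bones and hand-placed offset arrays; hiding feet halves PixieDust's partitions.
# User.Variant blends pink palettes toward blue.
# Spawn probability responds to music-reactive state, front-end context and distance, while Hearts' brightness and opacity also factor in platform and time of day.
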NiagaraSystem /BRCosmetics/Characters/Player/Female/Medium/Bodies/F_MED_DimeBlanketGrace/FX/NS_Character_DimeBlanketGrace_Idle
dependencies {
 usedDataInterfaces [
  &NiagaraDataInterfaceArrayInt32
  &NiagaraDataInterfaceArrayFloat3
  &NiagaraDataInterfaceCurve
  &NiagaraDataInterfaceColorCurve
  &NiagaraDataInterfacePlatformSet
  &NiagaraDataInterfaceArrayColor
  &NiagaraDataInterfaceArrayFloat
  &NiagaraDataInterfaceVector2DCurve
  &NiagaraDataInterfaceSkeletalMesh
 ]
 usedDynamicInputs [
  &Multiply_Vector2D_ByFloat
  &MakeVector2D
  /Niagara/DynamicInputs/Lerp/SmoothLerpOverTime_Float
  &RemapRange
 ]
 usedModules [
  &self:SystemSpawnScript.SystemScriptSource.SystemScriptGraph.NiagaraNodeAssignment_0.SetVariables_#1
  &self:SystemSpawnScript.SystemScriptSource.SystemScriptGraph.NiagaraNodeAssignment_3.SetVariables_#2
  &SystemState
  &Timeline
  &self:SystemSpawnScript.SystemScriptSource.SystemScriptGraph.NiagaraNodeAssignment_1.SetVariables_#3
  &self:SystemSpawnScript.SystemScriptSource.SystemScriptGraph.NiagaraNodeAssignment_4.SetVariables_#4
  &self:SystemSpawnScript.SystemScriptSource.SystemScriptGraph.NiagaraNodeAssignment_5.SetVariables_#5
  &self:SystemSpawnScript.SystemScriptSource.SystemScriptGraph.NiagaraNodeAssignment_2.SetVariables_#6
  &self:SystemSpawnScript.SystemScriptSource.SystemScriptGraph.NiagaraNodeAssignment_6.SetVariables_#7
  &self:SystemSpawnScript.SystemScriptSource.SystemScriptGraph.NiagaraNodeAssignment_7.SetVariables_#8
 ]
}
emitters {
 Hearts {
  emitterUpdateScript [
   &EmitterState()
   &SpawnRate(
    Spawn Probability=&MakeCustomFloatFromBool(
     Bool=$User.MusicReactive
     True Float=1
     False Float=&FN_FrontEndFloatModifier(
      Gameplay=$System.DistanceSpawn
      Front End=&MakeCustomFloatFromBool(Bool=&SetBoolByFloatComparison(A=$System.Music_Reactive_FX_Preview, B=0), True Float=1, False Float=0.5)
     )
    )
    SpawnRate=35
    Use Spawn Probability=true
   )
  ]
  particleSpawnScript [
   &PartitionParticles(Number Of Partitions=4)
   &InitializeParticle(
    AdjustAlpha=true
    AdjustHue=true
    AdjustSaturation=true
    AdjustValue=true
    Color=&LerpLinearColors(
     StartColor=&RandomRangeLinearColor(Minimum=<rgba 1 0.9 0.95 0.7>, Maximum=<rgba 1 0.6522 0.8261 0.8>)
     EndColor=<rgba 0.6445 0.7157 0.8308 0.4>
     LerpFactor=$User.Variant
    )
    Color Mode=::NewEnumerator3
    Lifetime Max=0.5
    Lifetime Min=0.35
    Lifetime Mode=::NewEnumerator1
    Sprite Size Mode=::NewEnumerator1
    Sprite UV Mode=::NewEnumerator1
    Uniform Sprite Size Min=6
    UsePositionOffset=false
    Write Material Random=true
   )
   &CameraOffset(Camera Offset Amount=&FN_FrontEndFloatModifier(Gameplay=7, Front End=7))
  ]
  particleUpdateScript [
   &ParticleState()
   &SkeletalMeshLocation(
    Bone / Socket Index=&SelectIntFromArray(Direct Array Index=$Particles.PartitionParticles.Partition, Int Selection Array=DataInterface() {propertyValues {IntData [0,2,3,6]}})
    Bone Sampling Mode=::NewEnumerator4
    MaskSpawnGroup=false
    Sampled Position Offset=&Multiply_VectorByFloat(
     Vector=&MergeHelper_VectorSwitch(
      0=&SelectVectorFromArray(
       Array Sampling Mode=::NewEnumerator0
       Direct Array Index=?
       Vector Selection Array=DataInterface() {propertyValues {FloatData [
          <13.8469 7.7903 -3.881>
          <11.2522 8.3854 -2.0743>
          <10.3397 8.5802 -0.6137>
          <11.0025 8.4368 1.6921>
          <10.3777 -4.9459 -10.0311>
          <7.5025 -1.6645 -12.2894>
          <6.2481 -4.4029 -12.6107>
          <5.4331 -3.8398 -13.1274>
          <-0.101 -1.7905 -15.4275>
          <-0.3545 -2.7968 -15.4441>
          <-0.3156 -5.2691 -15.0497>
          <-4.2327 -2.7146 -16.5896>
          <0.4632 -7.3039 -14.1695>
          <4.083 -9.0283 -11.3393>
          <-4.0508 -6.3825 -15.8216>
          <-1.8343 -10.0234 -13.268>
          <-7.3646 -3.8226 -17.2063>
          <-3.2505 -8.7706 -14.503>
          <-4.3861 -9.194 -14.4676>
          <-9.9518 -6.557 -16.8761>
          <-9.7222 -5.4688 17.1742>
          <-6.7582 -6.4381 16.1954>
          <-0.7073 -1.5295 15.6363>
          <-0.1013 -4.999 14.9706>
          <3.0838 -6.1586 13.4009>
          <5.0737 -2.9987 13.3253>
          <10.5142 -1.9349 10.7329>
          <10.2344 -6.1589 9.101>
          <10.1766 -7.8136 -6.8958>
          <8.435 -9.5465 -3.8889>
          <6.9802 -9.3356 -7.8033>
          <3.9451 -11.8323 -5.9099>
          <2.1309 -12.4776 -6.7162>
          <1.2058 -12.4016 -8.2163>
          <-1.2154 -12.1878 -10.5013>
          <2.4331 -12.5229 -5.9125>
          <-2.4985 -13.6895 -7.3975>
          <-7.2603 -10.8552 -13.5934>
          <-9.8422 -9.4376 -14.9723>
          <-9.6695 -13.8397 -7.658>
          <-9.8314 -10.1038 14.2888>
          <-8.0182 -11.5789 12.418>
          <-6.478 -12.5344 11.0071>
          <-4.9425 -9.1912 14.4436>
          <-2.4995 -9.7296 13.5371>
          <-2.2394 -11.592 11.5182>
          <-0.3209 -10.4246 12.2455>
          <4.8281 -7.9872 11.4019>
          <6.0066 -10.0271 7.521>
          <2.0829 -12.6841 5.9065>
          <1.3123 -13.3997 1.9187>
          <-5.0761 -13.9759 7.1745>
          <-2.9982 -14.302 1.4352>
          <3.131 -12.7265 -0.4214>
          <-1.604 -14.1567 -1.006>
          <-3.2091 -14.3621 -0.036>
          <-2.5882 -14.0099 -5.5454>
          <-3.2256 -14.3187 -2.1867>
          <5.0056 -11.7399 1.0661>
          <8.5385 -9.7841 -1.4046>
          <8.1792 -9.9972 0.0058>
          <8.8254 -9.5912 2.5178>
          <-5.0841 -14.4716 -1.3809>
          <-7.844 -14.3138 -4.8715>
          <-7.7276 -14.4443 -3.3443>
          <10.486 -0.4951 -10.493>
          <7.0092 2.5221 -11.1027>
          <4.5869 1.8313 -12.7499>
          <-3.368 -0.9551 -16.3429>
          <-4.5636 -0.2533 -16.5386>
          <-1.1056 3.8463 -13.738>
          <-5.6199 -0.5946 -16.8647>
          <-6.4288 -0.1883 -16.9882>
          <-4.9084 2.4192 -15.7817>
          <-6.0333 2.6735 -15.9263>
          <-4.1733 5.1007 -13.4034>
          <-4.6061 5.3759 -13.1631>
          <-1.3649 5.2478 -12.1252>
          <-2.8156 6.2846 -11.0648>
          <1.5054 5.0505 -11.2282>
          <-0.2415 6.3336 -9.6834>
          <4.7451 4.2151 -10.8396>
          <2.7432 5.8273 -9.6081>
          <3.0698 7.2075 -5.9425>
          <4.552 6.8738 -6.4628>
          <5.5184 7.9575 1.324>
          <1.5748 7.2793 -6.0605>
          <0.8231 7.377 -5.9661>
          <-2.1886 7.7528 -4.3678>
          <4.2739 8.0658 -0.3327>
          <-1.3269 7.9566 -0.096>
          <-0.7142 7.6649 5.2411>
          <2.8472 7.6753 4.2266>
          <5.6035 6.2312 7.7557>
          <2.4072 6.1011 9.4177>
          <6.4162 3.5522 10.8605>
          <2.8215 2.1234 13.679>
          <0.7125 6.4074 9.2949>
          <-1.5225 -0.2379 15.7328>
          <-2.0363 0.9234 15.622>
          <0.2623 7.333 6.6163>
          <-0.3478 3.7767 13.4371>
          <-2.6185 2.6174 15.0072>
          <-4.7499 7.0112 9.9783>
          <-1.7297 7.8936 2.7572>
          <-3.2374 7.9113 2.3828>
          <-5.4619 7.7233 7.5227>
          <-5.3458 7.819 -5.0889>
          <-5.3711 7.5791 -7.5562>
          <-6.1299 7.979 -1.444>
          <-7.2049 7.9522 -4.9159>
          <-8.0005 7.9087 -7.0509>
          <-6.8197 8.0114 -0.9089>
          <-5.8884 8.0364 1.1899>
          <-9.7021 8.2004 5.093>
          <-6.5266 7.5816 9.2327>
          <-6.297 5.7733 13.1626>
          <-8.9275 6.2364 13.3982>
          <-4.6077 4.5642 14.154>
          <-4.7826 3.6919 14.8401>
          <-6.8002 2.6134 16.2004>
          <-7.8461 2.0446 16.8662>
          <-9.6966 3.4731 16.3386>
         ]}}
       Random Seed=0
      )
      1=&SelectVectorFromArray(
       Array Sampling Mode=::NewEnumerator0
       Direct Array Index=?
       Vector Selection Array=DataInterface() {propertyValues {FloatData [
          <-3.9023 2.6873 11.5008>
          <-6.7838 1.4379 11.443>
          <-6.4109 4.4725 10.7345>
          <-6.6933 6.1611 9.9269>
          <-9.0457 7.1165 9.1076>
          <-7.1705 7.6597 8.887>
          <-9.0854 7.3201 8.9521>
          <-2.6713 8.6563 9.0521>
          <0.488 7.7049 10.3895>
          <-1.0535 9.1409 9.1796>
          <-1.4345 9.8199 8.4946>
          <3.8778 9.7061 8.3869>
          <-2.3889 10.5391 6.9492>
          <-2.1427 10.8739 6.3977>
          <-5.1583 10.5547 5.154>
          <-3.3832 11.4097 2.5002>
          <5.057 10.1893 5.3667>
          <8.2527 8.5932 3.9918>
          <6.496 9.7662 2.0655>
          <7.8815 9.0003 1.5845>
          <9.1651 8.3448 0.8753>
          <8.1373 8.5572 -4.4411>
          <6.8039 9.3572 -4.5489>
          <4.659 10.6572 1.0436>
          <1.8375 11.5857 0.961>
          <2.4162 11.4181 -0.7585>
          <2.1323 11.4966 -2.5092>
          <-2.5396 11.574 0.9855>
          <-1.1669 11.6904 -1.11>
          <2.2204 11.4147 -4.5>
          <4.847 10.3403 -5.277>
          <4.6517 10.1218 -6.8643>
          <-5.3628 4.0002 -10.9441>
          <-5.9882 6.3185 -9.8738>
          <0.3045 7.5708 -10.4867>
          <3.1601 7.585 -10.5744>
          <3.3085 8.0857 -10.2394>
          <4.0128 9.6671 -8.4622>
          <2.6507 9.0309 -9.7175>
          <0.8011 11.2934 -6.1848>
          <-2.4848 11.195 -4.8649>
          <-0.5057 11.6444 -3.1579>
          <-6.2828 9.4578 -6.9867>
          <-7.3737 9.1497 -7.158>
          <-9.6589 7.7806 -8.3481>
          <-12.0752 9.2224 -5.4272>
          <-4.4164 11.3326 1.0992>
          <-6.5312 11.0621 0.503>
          <-8.2016 10.78 1.6255>
          <-11.9785 8.5353 6.6805>
          <-11.2155 8.9247 6.2955>
          <-6.4724 10.498 4.5889>
          <-8.8354 8.3201 7.9333>
          <-0.8267 -2.1695 -12.3859>
          <1.7968 -7.3655 -10.2388>
          <-1.8917 -4.8557 -11.0517>
          <-0.899 0.0588 12.3695>
          <-3.6767 -1.7263 11.8382>
          <-1.6279 -3.6228 11.6616>
          <-5.5181 -1.6861 11.6488>
          <-4.9302 -3.334 11.1962>
          <-8.6336 1.126 11.512>
          <-3.1876 -6.1993 9.6037>
          <-0.0768 -7.7107 8.7823>
          <0.3372 -4.6763 11.932>
          <-6.2066 -7.2672 -7.3213>
          <-4.9368 -7.8338 -6.0808>
          <-4.0173 -6.3566 -9.2635>
          <-5.8623 -5.9942 -9.4262>
          <-6.5258 -4.9101 -10.2753>
          <-5.9865 -3.669 -11.124>
          <-1.8627 0.5539 -12.2045>
         ]}}
       Random Seed=0
      )
      2=&SelectVectorFromArray(
       Array Sampling Mode=::NewEnumerator0
       Direct Array Index=?
       Vector Selection Array=DataInterface() {propertyValues {FloatData [
          <11.0799 4.7885 8.3785>
          <13.9311 13.9514 -6.0855>
          <13.4611 13.6087 -3.1623>
          <13.5752 13.0696 -0.7566>
          <13.5024 13.5263 -2.615>
          <11.0173 12.0414 2.2187>
          <11.0224 13.5814 -2.0375>
          <10.3924 14.2009 -6.7748>
          <8.8148 14.1764 -5.2037>
          <8.3843 -8.0665 -8.0794>
          <11.5218 -8.0813 -7.3365>
          <13.0465 -0.293 9.0073>
          <13.4783 -2.4695 8.4896>
          <10.5522 -4.6187 6.3905>
          <9.9362 -5.5212 5.4938>
          <13.0973 -7.9183 1.2716>
          <11.1213 -8.1687 -2.7946>
          <12.0462 -8.1629 -4.6889>
          <14.121 -8.3869 -5.2706>
          <9.1833 -8.0396 -1.0439>
          <10.1038 -7.7497 0.9484>
          <12.2979 -8.1266 -8.7801>
          <8.9214 -7.8828 0.0841>
         ]}}
       Random Seed=0
      )
      3=&SelectVectorFromArray(
       Array Sampling Mode=::NewEnumerator0
       Direct Array Index=?
       Vector Selection Array=DataInterface() {propertyValues {FloatData [
          <-10.124 -3.3474 -8.4604>
          <-11.1554 -2.3178 -8.722>
          <-10.1857 -7.4151 -7.239>
          <-11.0012 -5.1272 -8.45>
          <-11.8881 -9.3264 -5.972>
          <-10.9182 -12.4111 -1.712>
          <-10.1412 -12.9333 -0.6644>
          <-14.11 -9.1229 -6.4319>
          <-11.7058 -12.296 -1.8456>
          <-12.6423 -12.7367 -0.7923>
          <-11.0587 -14.0959 4.3702>
          <-8.3171 -13.9494 1.9112>
          <-10.3185 -14.3358 8.1026>
          <-12.8286 -14.1953 9.452>
          <-8.4149 8.062 7.4127>
          <-9.4681 7.9781 0.9827>
          <-10.8977 8.1807 2.2258>
          <-14.4292 8.4536 4.0055>
          <-10.8529 7.4403 -2.1448>
          <-10.5585 7.0973 -3.2488>
          <-10.6584 5.1904 -5.9986>
          <-11.5072 6.7975 -3.8606>
          <-12.4907 -0.3027 -8.9957>
          <-13.6261 0.927 -8.9644>
          <-14.403 4.9918 -6.8245>
         ]}}
       Random Seed=0
      )
      4=&SelectVectorFromArray(
       Array Sampling Mode=::NewEnumerator0
       Direct Array Index=?
       Vector Selection Array=DataInterface() {propertyValues {FloatData [
          <-10.124 -3.3474 -8.4604>
          <-11.1554 -2.3178 -8.722>
          <-10.1857 -7.4151 -7.239>
          <-11.0012 -5.1272 -8.45>
          <-11.8881 -9.3264 -5.972>
          <-10.9182 -12.4111 -1.712>
          <-10.1412 -12.9333 -0.6644>
          <-14.11 -9.1229 -6.4319>
          <-11.7058 -12.296 -1.8456>
          <-12.6423 -12.7367 -0.7923>
          <-11.0587 -14.0959 4.3702>
          <-8.3171 -13.9494 1.9112>
          <-10.3185 -14.3358 8.1026>
          <-12.8286 -14.1953 9.452>
          <-8.4149 8.062 7.4127>
          <-9.4681 7.9781 0.9827>
          <-10.8977 8.1807 2.2258>
          <-14.4292 8.4536 4.0055>
          <-10.8529 7.4403 -2.1448>
          <-10.5585 7.0973 -3.2488>
          <-10.6584 5.1904 -5.9986>
          <-11.5072 6.7975 -3.8606>
          <-12.4907 -0.3027 -8.9957>
          <-13.6261 0.927 -8.9644>
          <-14.403 4.9918 -6.8245>
         ]}}
       Random Seed=0
      )
      5=&SelectVectorFromArray(
       Array Sampling Mode=::NewEnumerator0
       Direct Array Index=?
       Vector Selection Array=DataInterface() {propertyValues {FloatData [
          <-23.4605 -3.3464 -6.5313>
          <-22.7987 -3.2758 -6.5734>
          <-20.6573 -2.2485 -6.5435>
          <-21.5796 -0.3645 -5.8509>
          <-15.6881 0.6302 -5.4908>
          <-14.8417 0.6457 -5.5956>
          <-15.5633 -1.4272 -6.5982>
          <-19.6065 -4.7143 -6.6835>
          <-19.6623 -5.2141 -6.5411>
          <-16.6204 -8.2655 -4.9145>
          <-15.7158 -6.3715 -6.379>
         ]}}
       Random Seed=0
      )
      6=&SelectVectorFromArray(
       Array Sampling Mode=::NewEnumerator0
       Direct Array Index=?
       Vector Selection Array=DataInterface() {propertyValues {FloatData [
          <-10.124 -3.3474 -8.4604>
          <-11.1554 -2.3178 -8.722>
          <-10.1857 -7.4151 -7.239>
          <-11.0012 -5.1272 -8.45>
          <-11.8881 -9.3264 -5.972>
          <-10.9182 -12.4111 -1.712>
          <-10.1412 -12.9333 -0.6644>
          <-14.11 -9.1229 -6.4319>
          <-11.7058 -12.296 -1.8456>
          <-12.6423 -12.7367 -0.7923>
          <-11.0587 -14.0959 4.3702>
          <-8.3171 -13.9494 1.9112>
          <-10.3185 -14.3358 8.1026>
          <-12.8286 -14.1953 9.452>
          <-8.4149 8.062 7.4127>
          <-9.4681 7.9781 0.9827>
          <-10.8977 8.1807 2.2258>
          <-14.4292 8.4536 4.0055>
          <-10.8529 7.4403 -2.1448>
          <-10.5585 7.0973 -3.2488>
          <-10.6584 5.1904 -5.9986>
          <-11.5072 6.7975 -3.8606>
          <-12.4907 -0.3027 -8.9957>
          <-13.6261 0.927 -8.9644>
          <-14.403 4.9918 -6.8245>
         ]}}
       Random Seed=0
      )
      7=&SelectVectorFromArray(
       Array Sampling Mode=::NewEnumerator0
       Direct Array Index=?
       Vector Selection Array=DataInterface() {propertyValues {FloatData [
          <-23.4605 -3.3464 -6.5313>
          <-22.9131 -4.4496 -6.4709>
          <-22.7987 -3.2758 -6.5734>
          <-22.128 -1.5867 -6.3193>
          <-20.6573 -2.2485 -6.5435>
          <-20.7378 -1.012 -6.1482>
          <-21.5796 -0.3645 -5.8509>
          <-17.56 -0.1061 -5.8105>
          <-15.6881 0.6302 -5.4908>
          <-14.8417 0.6457 -5.5956>
          <-15.0548 -0.1377 -6.0554>
          <-15.5633 -1.4272 -6.5982>
          <-18.3501 -1.1394 -6.2538>
          <-19.954 -0.6898 -6.0401>
          <-19.5667 -1.0141 -6.1776>
          <-19.6065 -4.7143 -6.6835>
          <-20.2765 -6.3364 -6.1073>
          <-19.6623 -5.2141 -6.5411>
          <-18.8565 -7.4426 -5.4282>
          <-17.3539 -6.7195 -6.0347>
          <-16.6204 -8.2655 -4.9145>
          <-15.7158 -6.3715 -6.379>
          <-15.3707 -5.5116 -6.6963>
         ]}}
       Random Seed=0
      )
      8=&SelectVectorFromArray(
       Array Sampling Mode=::NewEnumerator0
       Direct Array Index=?
       Vector Selection Array=DataInterface() {propertyValues {FloatData []}}
       Random Seed=0
      )
      Index=$Particles.PartitionParticles.Partition
      Num=::NewEnumerator2
     )
     Float=0.95
    )
    Skeletal Mesh=$System.Skeletal Mesh
   )
   &InheritVelocity(Inherited Velocity Amount Scale=<-1 -1 -0.5>, Speed Limit=true, UseSpeedThreshold=false)
   &DynamicMaterialParameters(
    Index 0 Param 1=&FloatFromCurve(
     FloatCurve=DataInterface() {propertyValues {Curve {keys [{interpMode=RCIM_Cubic, tangentMode=RCTM_User, time=0, value=0}, {interpMode=RCIM_Cubic, tangentMode=RCTM_User, time=1, value=1}]}}}
     Scale Curve=&RandomRangeFloat(Minimum=0.9, Maximum=1.1)
    )
    Module.Index0Float0Write=true
    Module.Index0Float1Write=true
    Module.Index0Float2Write=true
    Module.Index0Float3Write=true
   )
   &ScaleSpriteSize(Uniform Curve Sprite Scale=DataInterface() {propertyValues {Curve {keys [
        {arriveTangent=2.4606, interpMode=RCIM_Cubic, leaveTangent=2.4606, tangentMode=RCTM_User, time=0, value=0}
        {interpMode=RCIM_Cubic, tangentMode=RCTM_User, time=1, value=1}
       ]}}})
   &SolveForcesAndVelocity(Clamp Velocity=false, Limit Acceleration=false)
   &ScaleColor(
    Scale Alpha=&Multiply_Float(
     A=&Multiply_Float(
      A=&FloatFromCurve(
       FloatCurve=DataInterface() {propertyValues {Curve {keys [<at 0 1 RCIM_Cubic>,<at 1 0>]}}}
       Scale Curve=&RandomRangeFloat(Minimum=1, Maximum=1)
      )
      B=&MakeFloatFromLinearColor(LinearColor=$System.TODColorMultiplier, Channel:ENiagaraLinearColor_Channels=::NewEnumerator3)
     )
     B=&MakeCustomFloatFromBool(
      Bool=$User.MusicReactive
      True Float=1
      False Float=&MakeCustomFloatFromBool(
       Bool=&SetBoolByFloatComparison(A=$System.Music_Reactive_FX_Preview, B=0)
       True Float=&Multiply_Float(
        A=&Lerp_Float(
         A=&FN_FrontEndFloatModifier(
          Gameplay=&FN_PickFloatDependOnPlatform(TargetPlatformActive=1, TargetPlatformNoActive=0.8)
          Front End=0.6
         )
         B=&FN_FrontEndFloatModifier(
          Gameplay=&FN_PickFloatDependOnPlatform(TargetPlatformActive=3, TargetPlatformNoActive=1.5)
          Front End=1
         )
         Alpha=$System.Music_Reactive_FX_Preview
        )
        B=&FNTimeOfDayScalar(Night Scale=0.4, Morning Scale=0.75, Day Scale=0.75, Evening Scale=0.75)
       )
       False Float=&FN_FrontEndFloatModifier(
        Gameplay=0
        Front End=&FloatFromCurve(
         FloatCurve=DataInterface() {propertyValues {Curve {keys [<at 0 0 RCIM_Cubic>, {arriveTangent=1.362, interpMode=RCIM_Cubic, leaveTangent=-0.9886, tangentMode=RCTM_Break, time=0.5, value=1}, <at 1 0>]}}}
         CurveIndex=&Cosine(
          Normalized Angle=&Add_Float(A=&Divide_Float(A=$Engine.Time, B=5), B=0.5)
          Period=1
          Scale=0.5
          Bias=1
         )
        )
       )
      )
     )
    )
    Scale RGB=&Multiply_Vector(
     A=&Lerp_Vector(
      A=&MakeVectorFromLinearColorRGB(Color=&MergeHelper_ColorSwitch(
        0=&ColorFromCurve(ColorCurve=DataInterface() {propertyValues {
           AlphaCurve {keys [<at 0 1 RCIM_Cubic>,<at 1 0 RCIM_Cubic>]}
           BlueCurve {keys [<at 0 9.6056 RCIM_Cubic>,<at 0.2254 7.5>,<at 0.4718 3.8333>,<at 0.8284 0.8667 RCIM_Cubic>]}
           GreenCurve {keys [<at 0 4.2112 RCIM_Cubic>,<at 0.2254 4>,<at 0.4718 1.5>,<at 0.8284 0.2 RCIM_Cubic>]}
           RedCurve {keys [<at 0 15 RCIM_Cubic>,<at 0.2254 10>,<at 0.4718 5>,<at 0.8284 1 RCIM_Cubic>]}
          }})
        1=&ColorFromCurve(ColorCurve=DataInterface() {propertyValues {
           AlphaCurve {keys [<at 0 1 RCIM_Cubic>,<at 1 0 RCIM_Cubic>]}
           BlueCurve {keys [<at 0 6.1 RCIM_Cubic>,<at 0.2254 3.6>,<at 0.4718 1.6667>,<at 0.8284 0.2104 RCIM_Cubic>]}
           GreenCurve {keys [<at 0 6 RCIM_Cubic>,<at 0.2254 3.2>,<at 0.4718 1.2>,<at 0.8284 0.1052 RCIM_Cubic>]}
           RedCurve {keys [<at 0 12 RCIM_Cubic>,<at 0.2254 8>,<at 0.4718 4>,<at 0.8284 0.526 RCIM_Cubic>]}
          }})
        2=&ColorFromCurve(ColorCurve=DataInterface() {propertyValues {
           AlphaCurve {keys [<at 0 1 RCIM_Cubic>,<at 1 0 RCIM_Cubic>]}
           BlueCurve {keys [<at 0 5.8794 RCIM_Cubic>,<at 0.2254 4.4754>,<at 0.4718 1.8203>,<at 0.8284 0.1753 RCIM_Cubic>]}
           GreenCurve {keys [<at 0 6.1854 RCIM_Cubic>,<at 0.2254 4.6516>,<at 0.4718 1.9656>,<at 0.8284 0.2046 RCIM_Cubic>]}
           RedCurve {keys [<at 0 12 RCIM_Cubic>,<at 0.2254 8>,<at 0.4718 4>,<at 0.8284 0.526 RCIM_Cubic>]}
          }})
        3=?
        Index=&RandomRangeInt(Minimum=0, Maximum=2)
        Num=::NewEnumerator1
       ))
      B=&MakeVectorFromLinearColorRGB(Color=&MergeHelper_ColorSwitch(
        0=&ColorFromCurve(ColorCurve=DataInterface() {propertyValues {
           AlphaCurve {keys [<at 0 1 RCIM_Cubic>,<at 1 0 RCIM_Cubic>]}
           BlueCurve {keys [<at 0 12 RCIM_Cubic>,<at 0.2254 8>,<at 0.4718 4>,<at 0.8284 0.526 RCIM_Cubic>]}
           GreenCurve {keys [<at 0 10 RCIM_Cubic>,<at 0.2254 6>,<at 0.4718 2.6>,<at 0.8284 0.2806 RCIM_Cubic>]}
           RedCurve {keys [<at 0 6 RCIM_Cubic>,<at 0.2254 3.2>,<at 0.4718 1.2>,<at 0.8284 0.1052 RCIM_Cubic>]}
          }})
        1=&ColorFromCurve(ColorCurve=DataInterface() {propertyValues {
           AlphaCurve {keys [<at 0 1 RCIM_Cubic>,<at 1 0 RCIM_Cubic>]}
           BlueCurve {keys [<at 0 12 RCIM_Cubic>,<at 0.2254 8>,<at 0.4718 4>,<at 0.8284 0.526 RCIM_Cubic>]}
           GreenCurve {keys [<at 0 5.1946 RCIM_Cubic>,<at 0.2254 3.0871>,<at 0.4718 1.7436>,<at 0.8284 0.1052 RCIM_Cubic>]}
           RedCurve {keys [<at 0 5.1946 RCIM_Cubic>,<at 0.2254 3.4965>,<at 0.4718 2.0445>,<at 0.8284 0.1543 RCIM_Cubic>]}
          }})
        2=?
        3=?
        Index=&RandomRangeInt(Minimum=0, Maximum=1)
        Num=::NewEnumerator0
       ))
      Alpha=$User.Variant
     )
     B=&MakeVectorFromLinearColorRGB(Color=$System.TODColorMultiplier)
    )
    ScaleA=true
    ScaleRGB=true
   )
  ]
  rendererClasses [&NiagaraSpriteRendererProperties]
  renderers [{rendererClass=&NiagaraSpriteRendererProperties}]
 }
 PixieDust {
  emitterUpdateScript [
   &EmitterState()
   &SpawnRate(
    Spawn Probability=&MakeCustomFloatFromBool(
     Bool=$User.MusicReactive
     True Float=1
     False Float=&FN_FrontEndFloatModifier(
      Gameplay=$System.DistanceSpawn
      Front End=&MakeCustomFloatFromBool(Bool=&SetBoolByFloatComparison(A=$System.Music_Reactive_FX_Preview, B=0), True Float=1, False Float=0.5)
     )
    )
    SpawnRate=20
    Use Spawn Probability=true
   )
  ]
  particleSpawnScript [
   &PartitionParticles(Number Of Partitions=&MakeCustomIntFromBool(Bool=$User.bisFootHidden, True INT=4, False INT=8))
   &InitializeParticle(
    AdjustAlpha=true
    AdjustHue=true
    AdjustSaturation=true
    AdjustValue=true
    Color=&LerpLinearColors(
     StartColor=&SelectLinearColorFromArray(
      Array Sampling Mode=::NewEnumerator1
      Direct Array Index=$Particles.PartitionParticles.Partition
      Color Selection Array=DataInterface() {propertyValues {ColorData [
         <rgba 1 0.4551 0.7275 0.4>
         <rgba 1 0.4551 0.7275 0.4>
         <rgba 1 0.4551 0.7275 0.4>
         <rgba 1 0.4551 0.7275 0.4>
         <rgba 1 1 1 0.4>
         <rgba 1 1 1 0.4>
         <rgba 1 0.4551 0.7275 0.4>
         <rgba 1 0.4551 0.7275 0.4>
        ]}}
      Random Seed=?
     )
     EndColor=&SelectLinearColorFromArray(
      Array Sampling Mode=::NewEnumerator1
      Direct Array Index=$Particles.PartitionParticles.Partition
      Color Selection Array=DataInterface() {propertyValues {ColorData [
         <rgba 0.3584 0.5715 0.9159 0.4>
         <rgba 0.3584 0.5715 0.9159 0.4>
         <rgba 0.3584 0.5715 0.9159 0.4>
         <rgba 0.3584 0.5715 0.9159 0.4>
         <rgba 1 1 1 0.4>
         <rgba 1 1 1 0.4>
         <rgba 0.3584 0.5715 0.9159 0.4>
         <rgba 0.3584 0.5715 0.9159 0.4>
        ]}}
      Random Seed=?
     )
     LerpFactor=$User.Variant
    )
    Color Mode=::NewEnumerator3
    Lifetime=&FN_FrontEndFloatModifier(
     Gameplay=&UniformRangedFloat(Minimum=1, Maximum=1.3)
     Front End=&UniformRangedFloat(Minimum=1, Maximum=1.25)
    )
    Sprite Rotation Angle=&RandomRangeFloat(Minimum=0.4, Maximum=0.6)
    Sprite Rotation Mode=::NewEnumerator1
    Sprite Size Mode=::NewEnumerator1
    Uniform Sprite Size Max=30
    Uniform Sprite Size Min=20
    UsePositionOffset=false
    Value Range:Vector2f=<1 1.3>
    Write Material Random=true
   )
   &DynamicMaterialParameters(
    Index 0 Param 1=&RandomRangeFloat(Minimum=-0.05, Maximum=0.05)
    Index 0 Param 2=&RandomRangeFloat(Minimum=0, Maximum=32)
    Module.Index0Float0Write=true
    Module.Index0Float1Write=true
    Module.Index0Float2Write=true
    Module.Index0Float3Write=true
   )
   &AddVelocity(
    Velocity=&FN_FrontEndVector3Modifier(Gameplay=<0 0 0>, Front End=<0 0 1>)
    Velocity Speed Scale=&RandomRangeFloat(Minimum=10, Maximum=17)
   )
   &CameraOffset(Camera Offset Amount=5)
  ]
  particleUpdateScript [
   &ParticleState()
   &SkeletalMeshLocation(
    Bone / Socket Index=&SelectIntFromArray(Direct Array Index=$Particles.PartitionParticles.Partition, Int Selection Array=DataInterface() {propertyValues {IntData [0,2,3,6,4,7,4,7]}})
    Bone Sampling Mode=::NewEnumerator4
    MaskSpawnGroup=false
    Sampled Position Offset=&Multiply_VectorByFloat(
     Vector=&MergeHelper_VectorSwitch(
      0=&SelectVectorFromArray(
       Array Sampling Mode=::NewEnumerator0
       Direct Array Index=?
       Vector Selection Array=DataInterface() {propertyValues {FloatData [
          <7.0825 -3.209 -12.387>
          <4.4278 -2.6188 -13.5834>
          <3.1341 -3.9886 -14.0972>
          <-0.5955 -2.8487 -15.5259>
          <-3.1934 -4.4878 -16.1889>
          <-4.8086 -2.523 -16.7536>
          <-0.3873 -5.595 -14.9873>
          <-3.4406 -4.8015 -16.2101>
          <-7.7295 -2.2975 -17.4125>
          <-3.0992 -7.7103 -15.1021>
          <0.0712 -8.7212 -13.4286>
          <-1.3831 -9.2904 -13.616>
          <-6.6177 -8.802 -15.1078>
          <-9.5666 -3.2257 17.5678>
          <-8.5956 -3.1043 17.3639>
          <-6.9516 -7.4854 15.8062>
          <-3.8061 -3.2305 16.3423>
          <-1.582 -3.3296 15.739>
          <1.8411 -2.8574 14.6115>
          <2.738 -2.7163 14.2898>
          <4.6268 -5.0906 13.0473>
          <7.4005 -5.843 11.4199>
          <8.7543 -4.0089 11.1607>
          <9.0374 -5.6223 10.2047>
          <8.4967 -6.1511 10.2623>
          <10.115 -5.105 10.0742>
          <10.2288 -6.0229 9.2215>
          <4.4789 -10.3534 -9.0671>
          <-2.9056 -11.8316 -11.4767>
          <-7.7717 -11.663 -12.489>
          <-7.1758 -10.5708 -13.8301>
          <-11.0904 -8.3759 -16.0865>
          <-8.1213 -12.166 -11.76>
          <-9.8637 -12.2054 -11.67>
          <-9.1633 -12.3159 11.3158>
          <-6.9635 -12.8824 10.5564>
          <-5.4336 -8.9027 14.7277>
          <-2.1884 -9.9455 13.3133>
          <-4.6373 -12.3733 11.0779>
          <-3.1025 -11.6262 11.7074>
          <1.4209 -10.1913 11.601>
          <7.2449 -8.307 8.9641>
          <8.5531 4.8756 8.4908>
          <10.5055 -0.7599 -10.5468>
          <9.9639 1.9517 -10.1723>
          <7.8173 -1.027 -12.1452>
          <8.0966 3.3356 -10.1008>
          <5.5148 -0.0825 -12.873>
          <5.0435 -0.3266 -13.152>
          <6.2502 2.0546 -11.7754>
          <4.013 2.4981 -12.6084>
          <2.4046 0.6935 -14.2323>
          <1.2161 1.9363 -14.23>
          <-10.3468 -1.352 -17.8761>
          <-3.5443 6.105 -11.7077>
          <-2.0714 5.6536 -11.795>
          <0.2152 4.606 -12.4779>
          <-5.4349 7.2047 -9.6086>
          <1.7793 4.7122 -11.5739>
          <4.88 4.4792 -10.4713>
          <5.2538 5.5491 -9.0358>
          <3.3383 6.4055 8.3575>
          <5.5861 3.7372 11.1844>
          <3.065 1.2912 13.8966>
          <2.3103 1.7857 14.0811>
          <0.7225 1.3271 14.7082>
          <-1.2887 2.232 14.9377>
          <0.0058 6.2509 9.9247>
          <-4.6036 -0.6091 16.562>
          <-3.3413 1.9692 15.7027>
          <-2.5612 6.9026 9.1559>
          <-8.3862 7.2478 -11.0405>
          <-9.3572 7.6319 -10.1582>
          <-6.3674 7.1378 10.5398>
          <-7.37 5.516 13.9443>
          <-6.0205 2.137 16.3366>
          <-4.3679 1.4493 16.1086>
          <-8.3429 4.5971 15.2158>
          <-7.8386 2.9104 16.2463>
          <-6.7778 2.0021 16.6253>
          <-10.2878 2.2682 17.3126>
          <-8.5675 0.1192 17.3283>
         ]}}
       Random Seed=0
      )
      1=&SelectVectorFromArray(
       Array Sampling Mode=::NewEnumerator0
       Direct Array Index=?
       Vector Selection Array=DataInterface() {propertyValues {FloatData [
          <-3.9023 2.6873 11.5008>
          <-6.7838 1.4379 11.443>
          <-6.4109 4.4725 10.7345>
          <-6.6933 6.1611 9.9269>
          <-9.0457 7.1165 9.1076>
          <-7.1705 7.6597 8.887>
          <-9.0854 7.3201 8.9521>
          <-2.6713 8.6563 9.0521>
          <0.488 7.7049 10.3895>
          <-1.0535 9.1409 9.1796>
          <-1.4345 9.8199 8.4946>
          <3.8778 9.7061 8.3869>
          <-2.3889 10.5391 6.9492>
          <-2.1427 10.8739 6.3977>
          <-5.1583 10.5547 5.154>
          <-3.3832 11.4097 2.5002>
          <5.057 10.1893 5.3667>
          <8.2527 8.5932 3.9918>
          <6.496 9.7662 2.0655>
          <7.8815 9.0003 1.5845>
          <9.1651 8.3448 0.8753>
          <8.1373 8.5572 -4.4411>
          <6.8039 9.3572 -4.5489>
          <4.659 10.6572 1.0436>
          <1.8375 11.5857 0.961>
          <2.4162 11.4181 -0.7585>
          <2.1323 11.4966 -2.5092>
          <-2.5396 11.574 0.9855>
          <-1.1669 11.6904 -1.11>
          <2.2204 11.4147 -4.5>
          <4.847 10.3403 -5.277>
          <4.6517 10.1218 -6.8643>
          <-5.3628 4.0002 -10.9441>
          <-5.9882 6.3185 -9.8738>
          <0.3045 7.5708 -10.4867>
          <3.1601 7.585 -10.5744>
          <3.3085 8.0857 -10.2394>
          <4.0128 9.6671 -8.4622>
          <2.6507 9.0309 -9.7175>
          <0.8011 11.2934 -6.1848>
          <-2.4848 11.195 -4.8649>
          <-0.5057 11.6444 -3.1579>
          <-6.2828 9.4578 -6.9867>
          <-7.3737 9.1497 -7.158>
          <-9.6589 7.7806 -8.3481>
          <-12.0752 9.2224 -5.4272>
          <-4.4164 11.3326 1.0992>
          <-6.5312 11.0621 0.503>
          <-8.2016 10.78 1.6255>
          <-11.9785 8.5353 6.6805>
          <-11.2155 8.9247 6.2955>
          <-6.4724 10.498 4.5889>
          <-8.8354 8.3201 7.9333>
          <-0.8267 -2.1695 -12.3859>
          <1.7968 -7.3655 -10.2388>
          <-1.8917 -4.8557 -11.0517>
          <-0.899 0.0588 12.3695>
          <-3.6767 -1.7263 11.8382>
          <-1.6279 -3.6228 11.6616>
          <-5.5181 -1.6861 11.6488>
          <-4.9302 -3.334 11.1962>
          <-8.6336 1.126 11.512>
          <-3.1876 -6.1993 9.6037>
          <-0.0768 -7.7107 8.7823>
          <0.3372 -4.6763 11.932>
          <-6.2066 -7.2672 -7.3213>
          <-4.9368 -7.8338 -6.0808>
          <-4.0173 -6.3566 -9.2635>
          <-5.8623 -5.9942 -9.4262>
          <-6.5258 -4.9101 -10.2753>
          <-5.9865 -3.669 -11.124>
          <-1.8627 0.5539 -12.2045>
         ]}}
       Random Seed=0
      )
      2=&SelectVectorFromArray(
       Array Sampling Mode=::NewEnumerator0
       Direct Array Index=?
       Vector Selection Array=DataInterface() {propertyValues {FloatData [
          <11.0799 4.7885 8.3785>
          <13.4973 6.549 7.904>
          <11.7211 9.3977 5.7426>
          <13.6073 11.1797 3.7062>
          <9.8237 10.7428 4.3975>
          <10.5772 11.289 3.6009>
          <11.5862 0.7274 8.7843>
          <14.4665 -6.5609 5.4994>
          <14.2054 -5.6658 6.2858>
          <11.4632 -6.8222 3.8299>
          <13.5481 -7.986 1.1975>
          <10.6092 -7.3224 2.7593>
         ]}}
       Random Seed=0
      )
      3=&SelectVectorFromArray(
       Array Sampling Mode=::NewEnumerator0
       Direct Array Index=?
       Vector Selection Array=DataInterface() {propertyValues {FloatData [
          <-10.124 -3.3474 -8.4604>
          <-11.1554 -2.3178 -8.722>
          <-10.1857 -7.4151 -7.239>
          <-11.0012 -5.1272 -8.45>
          <-11.8881 -9.3264 -5.972>
          <-10.9182 -12.4111 -1.712>
          <-10.1412 -12.9333 -0.6644>
          <-14.11 -9.1229 -6.4319>
          <-11.7058 -12.296 -1.8456>
          <-12.6423 -12.7367 -0.7923>
          <-10.7391 7.4238 -2.1637>
          <-11.8566 8.0944 -0.1618>
          <-14.1764 7.6739 -2.4151>
          <-10.1686 2.803 -7.4976>
          <-9.6626 -0.3483 -8.3328>
          <-13.0431 3.9264 -7.3148>
          <-11.1604 -1.2927 -8.7557>
         ]}}
       Random Seed=0
      )
      4=&SelectVectorFromArray(
       Array Sampling Mode=::NewEnumerator0
       Direct Array Index=?
       Vector Selection Array=DataInterface() {propertyValues {FloatData [
          <33.0731 7.5506 0.7231>
          <29.7452 7.6631 0.6326>
          <26.2292 8.0939 0.012>
          <20.135 9.1532 0.2868>
          <28.0947 5.7015 -2.737>
          <29.053 3.7759 -3.0437>
          <29.5431 3.0671 -3.1061>
          <33.1731 2.0331 -2.4908>
          <35.1531 5.3481 -1.7869>
          <33.7886 1.4966 -2.246>
          <35.611 -0.4619 -0.682>
          <36.6516 4.7087 4.1971>
          <36.4074 6.9535 1.68>
          <36.1073 6.3541 3.0004>
          <31.0996 7.258 2.3636>
          <35.0702 1.9975 4.5571>
          <34.2393 4.2185 4.5957>
          <32.0921 4.2398 4.6575>
          <32.8597 -0.3809 3.035>
          <31.4508 2.7528 4.8927>
          <30.4655 2.2682 5.0189>
          <30.8087 5.0953 4.5383>
          <30.5251 6.4306 3.7202>
          <28.8323 7.4712 2.6488>
          <25.2146 8.4173 1.8118>
          <25.5012 6.5492 4.6078>
          <22.9148 8.7424 0.6839>
          <15.2401 9.4075 1.2897>
          <19.9816 8.8786 2.7878>
          <21.9822 6.4671 5.5143>
          <11.3419 8.8414 2.355>
          <12.5128 3.0817 6.6728>
          <12.0181 4.0861 6.5326>
          <10.8853 8.3244 3.5719>
          <10.0513 0.7842 5.7365>
          <10.1367 -0.2611 4.8115>
          <10.7784 -2.0879 -0.8239>
          <10.5092 0.2958 -4.0702>
          <10.0816 3.8075 -4.8018>
          <12.4391 5.4014 -4.764>
          <15.1808 -0.192 -3.6269>
          <14.6179 4.583 -4.7875>
          <16.6672 -0.9792 -2.7017>
          <15.98 -2.6678 -0.408>
          <19.0657 -0.277 -3.0589>
          <17.855 0.693 -3.8028>
          <11.9237 -2.8661 2.9882>
          <14.6981 -3.2093 1.5513>
          <21.9636 -0.8273 -2.0461>
          <13.5304 6.1088 6.2605>
          <13.8143 6.812 5.835>
          <14.19 8.2272 4.4871>
          <14.2386 8.5532 3.9907>
          <18.399 -2.7031 0.9248>
          <23.1318 -1.3757 3.4663>
          <25.1515 1.9947 5.6674>
          <29.1156 -1.3107 0.5526>
          <29.4933 -0.9371 -0.6685>
          <27.5522 -0.9541 -1.0392>
          <29.1761 -0.3703 -1.317>
          <29.5861 2.3564 -2.9101>
          <28.4339 1.7211 -2.8354>
          <25.8009 4.6223 -3.4143>
          <24.5462 4.5573 -3.6806>
          <18.2773 2.5683 -4.6003>
          <21.9723 5.1496 -4.0772>
          <18.355 3.0671 -4.5859>
          <17.089 4.5411 -4.6621>
          <20.8448 6.5695 -3.5716>
          <20.0404 6.9537 -3.3922>
          <17.0247 7.9591 -2.8425>
          <16.4913 8.5682 -2.1979>
          <13.3959 8.3353 -2.6102>
          <13.3425 5.774 -4.5016>
          <12.9252 8.9402 -1.1481>
          <9.324 7.9691 -0.8149>
         ]}}
       Random Seed=0
      )
      5=&SelectVectorFromArray(
       Array Sampling Mode=::NewEnumerator0
       Direct Array Index=?
       Vector Selection Array=DataInterface() {propertyValues {FloatData [
          <-30.8056 -7.459 -0.1441>
          <-30.2873 -7.515 0.0282>
          <-29.4345 -7.5049 -0.0008>
          <-32.7424 -5.1241 2.3974>
          <-32.8908 -3.0442 2.7988>
          <-35.2062 -4.2983 2.2671>
          <-33.8773 1.0606 -1.268>
          <-34.0296 0.7605 0.5947>
          <-31.8279 1.0791 -1.033>
          <-30.3448 0.6209 0.9106>
          <-30.3532 1.1324 -0.4486>
          <-28.0811 -5.1054 3.0286>
          <-27.0896 -2.6251 3.3228>
          <-27.6264 -3.6195 3.3212>
          <-25.4173 -7.8164 1.1105>
          <-25.5042 -7.6673 1.3659>
          <-24.7808 -6.6127 2.7336>
          <-21.865 -8.4677 1.2169>
          <-20.5399 -7.3519 2.9899>
          <-18.3502 -8.45 2.1675>
          <-19.5256 -6.1954 4.0366>
          <-21.4575 -4.2982 4.1881>
          <-17.6363 -4.127 4.6193>
          <-19.7343 -4.9033 4.3855>
          <-22.9649 -0.6419 3.0978>
          <-27.8086 0.7101 1.2126>
          <-21.3831 -0.0533 3.0043>
          <-27.1772 1.6772 -1.0069>
          <-23.8667 1.6775 0.4552>
          <-22.6663 1.8249 0.4646>
          <-26.4004 1.0258 -3.238>
          <-19.8154 1.8876 1.2559>
          <-21.3594 2.3884 -1.6467>
          <-18.1853 -0.0816 3.4849>
          <-17.2854 2.92 -1.2379>
          <-20.0518 2.4242 -0.6305>
          <-16.6079 -0.2562 3.7339>
          <-16.7236 -4.646 4.6949>
          <-16.4357 1.1954 2.5021>
          <-15.6372 -5.496 4.7122>
          <-16.2143 -7.1982 3.4689>
          <-15.6106 1.2334 2.5763>
          <-15.3318 2.5664 1.1089>
          <-12.3866 -8.9894 0.9685>
          <-10.1835 -8.7378 0.9408>
          <-10.1274 -8.2029 2.2532>
          <-11.8125 -7.1502 3.5602>
          <-13.134 -4.3395 4.8601>
          <-10.595 -2.3343 3.8624>
          <-10.2957 -2.1272 4.7182>
          <-13.2499 0.504 3.6538>
          <-12.8182 1.6559 2.6568>
          <-10.3198 0.8031 -5.0843>
          <-10.8959 1.8149 -4.4955>
          <-10.8998 0.0325 -5.673>
          <-10.0892 -5.2541 -6.1825>
          <-10.2369 -8.4521 -2.9359>
          <-12.2151 -9.0288 -2.2307>
          <-10.5154 -7.9738 -4.0852>
          <-28.7811 -7.7504 -1.3308>
          <-35.0471 -6.1655 -3.373>
          <-32.6786 -4.6474 -4.5608>
          <-32.4496 -0.8687 -4.3888>
          <-29.8207 0.9606 -2.5174>
          <-28.4 0.9706 -2.8082>
          <-29.2873 -0.6832 -4.4697>
          <-30.0497 -2.1402 -5.0186>
          <-28.4678 -5.5469 -4.873>
          <-27.1419 -1.1475 -5.1449>
          <-26.9172 0.26 -4.0518>
          <-25.4821 -0.0821 -4.6006>
          <-24.398 0.0342 -4.6711>
          <-24.5041 -7.4487 -3.893>
          <-22.9245 -5.0192 -6.1348>
          <-21.7266 -7.976 -3.9233>
          <-23.1316 -0.2206 -5.1476>
          <-19.5376 1.9158 -3.3992>
          <-11.9143 3.2844 -1.9979>
          <-12.4994 -1.2472 -6.1914>
          <-13.8551 -1.5662 -6.3604>
          <-15.646 -2.9579 -6.913>
          <-16.1968 -9.1231 -2.8494>
          <-13.1992 -8.5348 -3.8929>
          <-13.2585 -7.9988 -4.5444>
          <-13.2651 -3.9712 -6.6099>
         ]}}
       Random Seed=0
      )
      6=&SelectVectorFromArray(
       Array Sampling Mode=::NewEnumerator0
       Direct Array Index=?
       Vector Selection Array=DataInterface() {propertyValues {FloatData [
          <16.7638 -0.4495 5.4513>
          <15.2952 3.9006 7.0079>
          <15.7278 5.697 6.6225>
          <16.5738 6.6284 6.1448>
          <16.5069 5.7042 6.5745>
          <15.9676 3.2484 6.9647>
          <17.6158 3.8009 6.8201>
          <21.3556 5.5985 6.2773>
          <21.0581 2.5746 6.5875>
          <17.9085 2.4845 6.6858>
          <17.977 1.7315 6.5231>
          <21.1249 0.4677 5.9296>
         ]}}
       Random Seed=0
      )
      7=&SelectVectorFromArray(
       Array Sampling Mode=::NewEnumerator0
       Direct Array Index=?
       Vector Selection Array=DataInterface() {propertyValues {FloatData [
          <-23.4605 -3.3464 -6.5313>
          <-22.7987 -3.2758 -6.5734>
          <-20.6573 -2.2485 -6.5435>
          <-21.5796 -0.3645 -5.8509>
          <-15.6881 0.6302 -5.4908>
          <-14.8417 0.6457 -5.5956>
          <-15.5633 -1.4272 -6.5982>
          <-19.6065 -4.7143 -6.6835>
          <-19.6623 -5.2141 -6.5411>
          <-16.6204 -8.2655 -4.9145>
          <-15.7158 -6.3715 -6.379>
         ]}}
       Random Seed=0
      )
      8=&SelectVectorFromArray(
       Array Sampling Mode=::NewEnumerator0
       Direct Array Index=?
       Vector Selection Array=DataInterface() {propertyValues {FloatData []}}
       Random Seed=0
      )
      Index=$Particles.PartitionParticles.Partition
      Num=::NewEnumerator6
     )
     Float=0.95
    )
    Skeletal Mesh=$System.Skeletal Mesh
   ) {module=SkeletalMeshLocation001}
   &ScaleColor(
    Scale Alpha=&Multiply_Float(
     A=&Multiply_Float(
      A=&Multiply_Float(
       A=&FloatFromCurve(
        FloatCurve=DataInterface() {propertyValues {Curve {keys [{interpMode=RCIM_Cubic, tangentMode=RCTM_User, time=0, value=1}, {interpMode=RCIM_Cubic, tangentMode=RCTM_User, time=1, value=0}]}}}
        Scale Curve=&SelectFloatFromArray(Direct Array Index=$Engine.QualityLevel, Float Selection Array=DataInterface() {propertyValues {FloatData [0.3,0.6,0.8,1,1]}})
       )
       B=&Lerp_Float(A=1, B=5, Alpha=$System.AudioReactivity)
      )
      B=&MakeFloatFromLinearColor(LinearColor=$System.TODColorMultiplier, Channel:ENiagaraLinearColor_Channels=::NewEnumerator3)
     )
     B=&FN_FrontEndFloatModifier(
      Gameplay=1
      Front End=&MakeCustomFloatFromBool(
       Bool=&SetBoolByFloatComparison(A=$NPC.FortniteNPC.EventAudio_Amplitude, B=0)
       True Float=1
       False Float=&FloatFromCurve(
        FloatCurve=DataInterface() {propertyValues {Curve {keys [<at 0 0 RCIM_Cubic>, {arriveTangent=1.362, interpMode=RCIM_Cubic, leaveTangent=-0.9886, tangentMode=RCTM_Break, time=0.5, value=1}, <at 1 0>]}}}
        CurveIndex=&Cosine(
         Normalized Angle=&Add_Float(A=&Divide_Float(A=$Engine.Time, B=5), B=0.5)
         Period=1
         Scale=0.5
         Bias=1
        )
       )
      )
     )
    )
    Scale RGB=&MakeVectorFromLinearColorRGB(Color=$System.TODColorMultiplier)
    ScaleA=true
    ScaleRGB=true
   ) {module=ScaleColor001}
   &ScaleSpriteSize(Uniform Curve Sprite Scale=DataInterface() {propertyValues {Curve {keys [
        {arriveTangent=2.2332, interpMode=RCIM_Cubic, leaveTangent=2.2332, tangentMode=RCTM_User, time=0, value=0.6}
        {tangentMode=RCTM_User, time=1, value=1.25}
       ]}}})
   &SolveForcesAndVelocity(Clamp Velocity=false, Limit Acceleration=false)
  ]
  rendererClasses [&NiagaraSpriteRendererProperties]
  renderers [{rendererClass=&NiagaraSpriteRendererProperties}]
 }
 Sparkles_Pink {
  emitterUpdateScript [
   &EmitterState()
   &SpawnRate(
    Spawn Probability=&MakeCustomFloatFromBool(
     Bool=&SetBoolByFloatComparison(A=$System.Music_Reactive_FX_Preview, B=0)
     True Float=1
     False Float=$System.SpawnProbability
    )
    SpawnRate=40
    Use Spawn Probability=true
   )
  ]
  particleSpawnScript [
   &PartitionParticles(Number Of Partitions=&MakeCustomIntFromBool(Bool=$User.bisFootHidden, True INT=4, False INT=6))
   &InitializeParticle(
    AdjustAlpha=false
    AdjustHue=true
    AdjustSaturation=true
    AdjustValue=true
    Color=&LerpLinearColors(StartColor=<rgba 1.7669 0.6014 2 1>, EndColor=<rgba 0.2 1.91 2 1>, LerpFactor=$User.Variant)
    Color Mode=::NewEnumerator3
    Hue Shift Range:Vector2f=<-0.05 0.05>
    Lifetime Max=0.5
    Lifetime Min=0.35
    Lifetime Mode=::NewEnumerator1
    Sprite Size Mode=::NewEnumerator3
    Uniform Sprite Size=&Multiply_Float(
     A=&RandomRangeFloat(Minimum=15, Maximum=20)
     B=&SelectFloatFromArray(Direct Array Index=$Engine.QualityLevel, Float Selection Array=DataInterface() {propertyValues {FloatData [0.5,0.6,0.8,1,1]}})
    )
    UsePositionOffset=false
    Write Material Random=true
   )
   &CameraOffset(Camera Offset Amount=&FN_FrontEndFloatModifier(Gameplay=4, Front End=6))
  ]
  particleUpdateScript [
   &ParticleState()
   &SkeletalMeshLocation(
    Bone / Socket Index=&SelectIntFromArray(Direct Array Index=$Particles.PartitionParticles.Partition, Int Selection Array=DataInterface() {propertyValues {IntData [0,2,3,6,4,7]}})
    Bone Sampling Mode=::NewEnumerator4
    MaskSpawnGroup=false
    Sampled Position Offset=&Multiply_VectorByFloat(
     Vector=&MergeHelper_VectorSwitch(
      0=&SelectVectorFromArray(
       Array Sampling Mode=::NewEnumerator0
       Direct Array Index=?
       Vector Selection Array=DataInterface() {propertyValues {FloatData [
          <13.8469 7.7903 -3.881>
          <11.2522 8.3854 -2.0743>
          <10.3397 8.5802 -0.6137>
          <11.0025 8.4368 1.6921>
          <10.3777 -4.9459 -10.0311>
          <7.5025 -1.6645 -12.2894>
          <6.2481 -4.4029 -12.6107>
          <5.4331 -3.8398 -13.1274>
          <-0.101 -1.7905 -15.4275>
          <-0.3545 -2.7968 -15.4441>
          <-0.3156 -5.2691 -15.0497>
          <-4.2327 -2.7146 -16.5896>
          <0.4632 -7.3039 -14.1695>
          <4.083 -9.0283 -11.3393>
          <-4.0508 -6.3825 -15.8216>
          <-1.8343 -10.0234 -13.268>
          <-7.3646 -3.8226 -17.2063>
          <-3.2505 -8.7706 -14.503>
          <-4.3861 -9.194 -14.4676>
          <-9.9518 -6.557 -16.8761>
          <-9.7222 -5.4688 17.1742>
          <-6.7582 -6.4381 16.1954>
          <-0.7073 -1.5295 15.6363>
          <-0.1013 -4.999 14.9706>
          <3.0838 -6.1586 13.4009>
          <5.0737 -2.9987 13.3253>
          <10.5142 -1.9349 10.7329>
          <10.2344 -6.1589 9.101>
          <10.1766 -7.8136 -6.8958>
          <8.435 -9.5465 -3.8889>
          <6.9802 -9.3356 -7.8033>
          <3.9451 -11.8323 -5.9099>
          <2.1309 -12.4776 -6.7162>
          <1.2058 -12.4016 -8.2163>
          <-1.2154 -12.1878 -10.5013>
          <2.4331 -12.5229 -5.9125>
          <-2.4985 -13.6895 -7.3975>
          <-7.2603 -10.8552 -13.5934>
          <-9.8422 -9.4376 -14.9723>
          <-9.6695 -13.8397 -7.658>
          <-9.8314 -10.1038 14.2888>
          <-8.0182 -11.5789 12.418>
          <-6.478 -12.5344 11.0071>
          <-4.9425 -9.1912 14.4436>
          <-2.4995 -9.7296 13.5371>
          <-2.2394 -11.592 11.5182>
          <-0.3209 -10.4246 12.2455>
          <4.8281 -7.9872 11.4019>
          <6.0066 -10.0271 7.521>
          <2.0829 -12.6841 5.9065>
          <1.3123 -13.3997 1.9187>
          <-5.0761 -13.9759 7.1745>
          <-2.9982 -14.302 1.4352>
          <3.131 -12.7265 -0.4214>
          <-1.604 -14.1567 -1.006>
          <-3.2091 -14.3621 -0.036>
          <-2.5882 -14.0099 -5.5454>
          <-3.2256 -14.3187 -2.1867>
          <5.0056 -11.7399 1.0661>
          <8.5385 -9.7841 -1.4046>
          <8.1792 -9.9972 0.0058>
          <8.8254 -9.5912 2.5178>
          <-5.0841 -14.4716 -1.3809>
          <-7.844 -14.3138 -4.8715>
          <-7.7276 -14.4443 -3.3443>
          <10.486 -0.4951 -10.493>
          <7.0092 2.5221 -11.1027>
          <4.5869 1.8313 -12.7499>
          <-3.368 -0.9551 -16.3429>
          <-4.5636 -0.2533 -16.5386>
          <-1.1056 3.8463 -13.738>
          <-5.6199 -0.5946 -16.8647>
          <-6.4288 -0.1883 -16.9882>
          <-4.9084 2.4192 -15.7817>
          <-6.0333 2.6735 -15.9263>
          <-4.1733 5.1007 -13.4034>
          <-4.6061 5.3759 -13.1631>
          <-1.3649 5.2478 -12.1252>
          <-2.8156 6.2846 -11.0648>
          <1.5054 5.0505 -11.2282>
          <-0.2415 6.3336 -9.6834>
          <4.7451 4.2151 -10.8396>
          <2.7432 5.8273 -9.6081>
          <3.0698 7.2075 -5.9425>
          <4.552 6.8738 -6.4628>
          <5.5184 7.9575 1.324>
          <1.5748 7.2793 -6.0605>
          <0.8231 7.377 -5.9661>
          <-2.1886 7.7528 -4.3678>
          <4.2739 8.0658 -0.3327>
          <-1.3269 7.9566 -0.096>
          <-0.7142 7.6649 5.2411>
          <2.8472 7.6753 4.2266>
          <5.6035 6.2312 7.7557>
          <2.4072 6.1011 9.4177>
          <6.4162 3.5522 10.8605>
          <2.8215 2.1234 13.679>
          <0.7125 6.4074 9.2949>
          <-1.5225 -0.2379 15.7328>
          <-2.0363 0.9234 15.622>
          <0.2623 7.333 6.6163>
          <-0.3478 3.7767 13.4371>
          <-2.6185 2.6174 15.0072>
          <-4.7499 7.0112 9.9783>
          <-1.7297 7.8936 2.7572>
          <-3.2374 7.9113 2.3828>
          <-5.4619 7.7233 7.5227>
          <-5.3458 7.819 -5.0889>
          <-5.3711 7.5791 -7.5562>
          <-6.1299 7.979 -1.444>
          <-7.2049 7.9522 -4.9159>
          <-8.0005 7.9087 -7.0509>
          <-6.8197 8.0114 -0.9089>
          <-5.8884 8.0364 1.1899>
          <-9.7021 8.2004 5.093>
          <-6.5266 7.5816 9.2327>
          <-6.297 5.7733 13.1626>
          <-8.9275 6.2364 13.3982>
          <-4.6077 4.5642 14.154>
          <-4.7826 3.6919 14.8401>
          <-6.8002 2.6134 16.2004>
          <-7.8461 2.0446 16.8662>
          <-9.6966 3.4731 16.3386>
         ]}}
       Random Seed=0
      )
      1=&SelectVectorFromArray(
       Array Sampling Mode=::NewEnumerator0
       Direct Array Index=?
       Vector Selection Array=DataInterface() {propertyValues {FloatData [
          <-3.9023 2.6873 11.5008>
          <-6.7838 1.4379 11.443>
          <-6.4109 4.4725 10.7345>
          <-6.6933 6.1611 9.9269>
          <-9.0457 7.1165 9.1076>
          <-7.1705 7.6597 8.887>
          <-9.0854 7.3201 8.9521>
          <-2.6713 8.6563 9.0521>
          <0.488 7.7049 10.3895>
          <-1.0535 9.1409 9.1796>
          <-1.4345 9.8199 8.4946>
          <3.8778 9.7061 8.3869>
          <-2.3889 10.5392 6.9492>
          <-2.1427 10.8739 6.3977>
          <-5.1583 10.5547 5.154>
          <-3.3832 11.4097 2.5002>
          <5.057 10.1893 5.3667>
          <8.2527 8.5932 3.9918>
          <6.496 9.7662 2.0655>
          <7.8815 9.0003 1.5845>
          <9.1651 8.3448 0.8753>
          <8.1373 8.5572 -4.4411>
          <6.8039 9.3572 -4.5489>
          <4.659 10.6572 1.0436>
          <1.8375 11.5857 0.961>
          <2.4162 11.4181 -0.7585>
          <2.1323 11.4966 -2.5092>
          <-2.5396 11.574 0.9855>
          <-1.1669 11.6904 -1.11>
          <2.2204 11.4147 -4.5>
          <4.847 10.3403 -5.277>
          <4.6517 10.1218 -6.8643>
          <-5.3628 4.0002 -10.9441>
          <-5.9882 6.3185 -9.8738>
          <0.3045 7.5708 -10.4867>
          <3.1601 7.585 -10.5744>
          <3.3085 8.0857 -10.2394>
          <4.0128 9.6671 -8.4622>
          <2.6507 9.0309 -9.7175>
          <0.8011 11.2934 -6.1848>
          <-2.4848 11.195 -4.8649>
          <-0.5057 11.6444 -3.1579>
          <-6.2828 9.4578 -6.9867>
          <-7.3737 9.1497 -7.158>
          <-9.6589 7.7806 -8.3481>
          <-12.0752 9.2224 -5.4272>
          <-4.4164 11.3326 1.0992>
          <-6.5312 11.0621 0.503>
          <-8.2016 10.78 1.6255>
          <-11.9785 8.5353 6.6805>
          <-11.2155 8.9247 6.2955>
          <-6.4724 10.498 4.5889>
          <-8.8354 8.3201 7.9333>
          <-0.8267 -2.1695 -12.3859>
          <1.7968 -7.3655 -10.2388>
          <-1.8917 -4.8557 -11.0517>
          <-0.899 0.0588 12.3695>
          <-3.6767 -1.7263 11.8382>
          <-1.6279 -3.6228 11.6616>
          <-5.5181 -1.6861 11.6488>
          <-4.9302 -3.334 11.1962>
          <-8.6336 1.126 11.512>
          <-3.1876 -6.1993 9.6037>
          <-0.0768 -7.7107 8.7823>
          <0.3372 -4.6763 11.932>
          <-6.2066 -7.2672 -7.3213>
          <-4.9368 -7.8338 -6.0808>
          <-4.0173 -6.3566 -9.2635>
          <-5.8623 -5.9942 -9.4262>
          <-6.5258 -4.9101 -10.2753>
          <-5.9865 -3.669 -11.124>
          <-1.8627 0.5539 -12.2045>
         ]}}
       Random Seed=0
      )
      2=&SelectVectorFromArray(
       Array Sampling Mode=::NewEnumerator0
       Direct Array Index=?
       Vector Selection Array=DataInterface() {propertyValues {FloatData [
          <11.0799 4.7885 8.3785>
          <13.9311 13.9514 -6.0855>
          <13.4611 13.6087 -3.1623>
          <13.5752 13.0696 -0.7566>
          <13.5024 13.5263 -2.615>
          <11.0173 12.0414 2.2187>
          <11.0224 13.5814 -2.0375>
          <10.3924 14.2009 -6.7748>
          <8.8148 14.1764 -5.2037>
          <8.3843 -8.0665 -8.0794>
          <11.5218 -8.0813 -7.3365>
          <13.0465 -0.293 9.0073>
          <13.4783 -2.4695 8.4896>
          <10.5522 -4.6187 6.3905>
          <9.9362 -5.5212 5.4938>
          <13.0973 -7.9183 1.2716>
          <11.1213 -8.1687 -2.7946>
          <12.0462 -8.1629 -4.6889>
          <14.121 -8.3869 -5.2706>
          <9.1833 -8.0396 -1.0439>
          <10.1038 -7.7497 0.9484>
          <12.2979 -8.1266 -8.7801>
          <8.9214 -7.8828 0.0841>
         ]}}
       Random Seed=0
      )
      3=&SelectVectorFromArray(
       Array Sampling Mode=::NewEnumerator0
       Direct Array Index=?
       Vector Selection Array=DataInterface() {propertyValues {FloatData [
          <-10.124 -3.3474 -8.4604>
          <-11.1554 -2.3178 -8.722>
          <-10.1857 -7.4151 -7.239>
          <-11.0012 -5.1272 -8.45>
          <-11.8881 -9.3264 -5.972>
          <-10.9182 -12.4111 -1.712>
          <-10.1412 -12.9333 -0.6644>
          <-14.11 -9.1229 -6.4319>
          <-11.7058 -12.296 -1.8456>
          <-12.6423 -12.7367 -0.7923>
          <-11.0587 -14.0959 4.3702>
          <-8.3171 -13.9494 1.9112>
          <-10.3185 -14.3358 8.1026>
          <-12.8286 -14.1953 9.452>
          <-8.4149 8.062 7.4127>
          <-9.4681 7.9781 0.9827>
          <-10.8977 8.1807 2.2258>
          <-14.4292 8.4536 4.0055>
          <-10.8529 7.4403 -2.1448>
          <-10.5585 7.0973 -3.2488>
          <-10.6584 5.1904 -5.9986>
          <-11.5072 6.7975 -3.8606>
          <-12.4907 -0.3027 -8.9957>
          <-13.6261 0.927 -8.9644>
          <-14.403 4.9918 -6.8245>
         ]}}
       Random Seed=0
      )
      4=&SelectVectorFromArray(
       Array Sampling Mode=::NewEnumerator0
       Direct Array Index=?
       Vector Selection Array=DataInterface() {propertyValues {FloatData [
          <16.7638 -0.4495 5.4513>
          <15.2952 3.9006 7.0079>
          <15.7278 5.697 6.6225>
          <16.5738 6.6284 6.1448>
          <16.5069 5.7042 6.5745>
          <15.9676 3.2484 6.9647>
          <17.6158 3.8009 6.8201>
          <21.3556 5.5985 6.2773>
          <21.0581 2.5746 6.5875>
          <17.9085 2.4845 6.6858>
          <17.977 1.7315 6.5231>
          <21.1249 0.4677 5.9296>
         ]}}
       Random Seed=0
      )
      5=&SelectVectorFromArray(
       Array Sampling Mode=::NewEnumerator0
       Direct Array Index=?
       Vector Selection Array=DataInterface() {propertyValues {FloatData [
          <-23.4605 -3.3464 -6.5313>
          <-22.7987 -3.2758 -6.5734>
          <-20.6573 -2.2485 -6.5435>
          <-21.5796 -0.3645 -5.8509>
          <-15.6881 0.6302 -5.4908>
          <-14.8417 0.6457 -5.5956>
          <-15.5633 -1.4272 -6.5982>
          <-19.6065 -4.7143 -6.6835>
          <-19.6623 -5.2141 -6.5411>
          <-16.6204 -8.2655 -4.9145>
          <-15.7158 -6.3715 -6.379>
         ]}}
       Random Seed=0
      )
      6=&SelectVectorFromArray(
       Array Sampling Mode=::NewEnumerator0
       Direct Array Index=?
       Vector Selection Array=DataInterface() {propertyValues {FloatData [
          <-10.124 -3.3474 -8.4604>
          <-11.1554 -2.3178 -8.722>
          <-10.1857 -7.4151 -7.239>
          <-11.0012 -5.1272 -8.45>
          <-11.8881 -9.3264 -5.972>
          <-10.9182 -12.4111 -1.712>
          <-10.1412 -12.9333 -0.6644>
          <-14.11 -9.1229 -6.4319>
          <-11.7058 -12.296 -1.8456>
          <-12.6423 -12.7367 -0.7923>
          <-11.0587 -14.0959 4.3702>
          <-8.3171 -13.9494 1.9112>
          <-10.3185 -14.3358 8.1026>
          <-12.8286 -14.1953 9.452>
          <-8.4149 8.062 7.4127>
          <-9.4681 7.9781 0.9827>
          <-10.8977 8.1807 2.2258>
          <-14.4292 8.4536 4.0055>
          <-10.8529 7.4403 -2.1448>
          <-10.5585 7.0973 -3.2488>
          <-10.6584 5.1904 -5.9986>
          <-11.5072 6.7975 -3.8606>
          <-12.4907 -0.3027 -8.9957>
          <-13.6261 0.927 -8.9644>
          <-14.403 4.9918 -6.8245>
         ]}}
       Random Seed=0
      )
      7=&SelectVectorFromArray(
       Array Sampling Mode=::NewEnumerator0
       Direct Array Index=?
       Vector Selection Array=DataInterface() {propertyValues {FloatData [
          <-23.4605 -3.3464 -6.5313>
          <-22.9131 -4.4496 -6.4709>
          <-22.7987 -3.2758 -6.5734>
          <-22.128 -1.5867 -6.3193>
          <-20.6573 -2.2485 -6.5435>
          <-20.7378 -1.012 -6.1482>
          <-21.5796 -0.3645 -5.8509>
          <-17.56 -0.1061 -5.8105>
          <-15.6881 0.6302 -5.4908>
          <-14.8417 0.6457 -5.5956>
          <-15.0548 -0.1377 -6.0554>
          <-15.5633 -1.4272 -6.5982>
          <-18.3501 -1.1394 -6.2538>
          <-19.954 -0.6898 -6.0401>
          <-19.5667 -1.0141 -6.1776>
          <-19.6065 -4.7143 -6.6835>
          <-20.2765 -6.3364 -6.1073>
          <-19.6623 -5.2141 -6.5411>
          <-18.8565 -7.4426 -5.4282>
          <-17.3539 -6.7195 -6.0347>
          <-16.6204 -8.2655 -4.9145>
          <-15.7158 -6.3715 -6.379>
          <-15.3707 -5.5116 -6.6963>
         ]}}
       Random Seed=0
      )
      8=&SelectVectorFromArray(
       Array Sampling Mode=::NewEnumerator0
       Direct Array Index=?
       Vector Selection Array=DataInterface() {propertyValues {FloatData []}}
       Random Seed=0
      )
      Index=$Particles.PartitionParticles.Partition
      Num=::NewEnumerator4
     )
     Float=0.95
    )
    Skeletal Mesh=$System.Skeletal Mesh
   )
   &ScaleSpriteSize(
    Uniform Curve Scale=&FNTimeOfDayScalar(Night Scale=1.25, Morning Scale=1, Evening Scale=1)
    Uniform Curve Sprite Scale=DataInterface() {propertyValues {Curve {keys [
        {interpMode=RCIM_Cubic, tangentMode=RCTM_User, time=0, value=0}
        {interpMode=RCIM_Cubic, tangentMode=RCTM_User, time=0.5, value=1}
        {tangentMode=RCTM_User, time=1, value=0}
       ]}}}
   ) {module=ScaleSpriteSize001}
   &ScaleColor(
    Scale Alpha=&Multiply_Float(
     A=&Multiply_Float(
      A=&FloatFromCurve(
       FloatCurve=DataInterface() {propertyValues {Curve {keys [
           {arriveTangent=6.5707, interpMode=RCIM_Cubic, leaveTangent=6.5707, tangentMode=RCTM_User, time=0, value=0}
           {interpMode=RCIM_Cubic, tangentMode=RCTM_User, time=0.3, value=1}
           {tangentMode=RCTM_User, time=1, value=0}
          ]}}}
       Scale Curve=&RandomRangeFloat(Minimum=0.1, Maximum=0.5)
      )
      B=&MakeFloatFromLinearColor(LinearColor=$System.TODColorMultiplier, Channel:ENiagaraLinearColor_Channels=::NewEnumerator3)
     )
     B=&FN_FrontEndFloatModifier(
      Gameplay=1
      Front End=&MakeCustomFloatFromBool(
       Bool=&SetBoolByFloatComparison(A=$NPC.FortniteNPC.EventAudio_Amplitude, B=0)
       True Float=1
       False Float=&FloatFromCurve(
        FloatCurve=DataInterface() {propertyValues {Curve {keys [<at 0 0 RCIM_Cubic>, {arriveTangent=1.362, interpMode=RCIM_Cubic, leaveTangent=-0.9886, tangentMode=RCTM_Break, time=0.5, value=1}, <at 1 0>]}}}
        CurveIndex=&Cosine(
         Normalized Angle=&Add_Float(A=&Divide_Float(A=$Engine.Time, B=5), B=0.5)
         Period=1
         Scale=0.5
         Bias=1
        )
       )
      )
     )
    )
    Scale RGB=&Multiply_Vector(
     A=&VectorFromFloat(Value=&Multiply_Float(
       A=&RandomRangeFloat(Minimum=0.75, Maximum=1)
       B=&SelectFloatFromArray(Direct Array Index=$Engine.QualityLevel, Float Selection Array=DataInterface() {propertyValues {FloatData [10,5,1,1,1]}})
      ))
     B=&MakeVectorFromLinearColorRGB(Color=$System.TODColorMultiplier)
    )
    ScaleA=true
    ScaleRGB=true
   ) {module=ScaleColor001}
   /SetVariables(
    Particles.CenterFlareColor:LinearColor=&MultiplyLinearColor(
     Linear Color=$Particles.Color
     Scale Factor=&MakeVector4_FromLinearColor(LinearColor=&LerpLinearColors(StartColor=<rgba 10 10 10 1>, EndColor=<rgba 10 10 10 1>, LerpFactor=$User.Variant))
    )
    Particles.CenterFlareSize:Vector2f=&Multiply_Vector2(A=$Particles.SpriteSize, B=&Vector2DFromFloat(Value=0.5))
    Particles.CoronaFlareColor:LinearColor=&MultiplyLinearColor(
     Linear Color=$Particles.Color
     Scale Factor=&MakeVector4_FromLinearColor(LinearColor=&LerpLinearColors(StartColor=<rgba 0.15 0.15 0.15 1>, EndColor=<rgba 0.0729 0.0729 0.0729 1>, LerpFactor=$User.Variant))
    )
    Particles.CoronaFlareSize:Vector2f=&Multiply_Vector2(A=$Particles.Initial.SpriteSize, B=&Vector2DFromFloat(Value=0.75))
   ) {isSetParameters=true, module=SetVariables_#9}
  ]
  rendererClasses [&NiagaraSpriteRendererProperties]
  renderers [
   {rendererClass=&NiagaraSpriteRendererProperties}
   {rendererClass=&NiagaraSpriteRendererProperties, rendererIndex=1}
  ]
 }
 Sparkles_Silver {
  emitterUpdateScript [
   &EmitterState()
   &SpawnRate(
    Spawn Probability=&MakeCustomFloatFromBool(
     Bool=&SetBoolByFloatComparison(A=$System.Music_Reactive_FX_Preview, B=0)
     True Float=1
     False Float=$System.SpawnProbability
    )
    SpawnRate=40
    Use Spawn Probability=true
   )
  ]
  particleSpawnScript [
   &PartitionParticles(Number Of Partitions=&MakeCustomIntFromBool(Bool=$User.bisFootHidden, True INT=4, False INT=8))
   &InitializeParticle(
    AdjustAlpha=false
    AdjustHue=true
    AdjustSaturation=true
    AdjustValue=true
    Color=<rgba 1.9897 1.9072 2 1>
    Color Mode=::NewEnumerator3
    Hue Shift Range:Vector2f=<-0.05 0.05>
    Lifetime Max=0.5
    Lifetime Min=0.35
    Lifetime Mode=::NewEnumerator1
    Sprite Size Mode=::NewEnumerator3
    Uniform Sprite Size=&Multiply_Float(
     A=&RandomRangeFloat(Minimum=15, Maximum=20)
     B=&SelectFloatFromArray(Direct Array Index=$Engine.QualityLevel, Float Selection Array=DataInterface() {propertyValues {FloatData [0.5,0.6,0.8,1,1]}})
    )
    UsePositionOffset=false
    Write Material Random=true
   )
   &CameraOffset(Camera Offset Amount=&FN_FrontEndFloatModifier(Gameplay=4, Front End=6))
  ]
  particleUpdateScript [
   &ParticleState()
   &SkeletalMeshLocation(
    Bone / Socket Index=&SelectIntFromArray(Direct Array Index=$Particles.PartitionParticles.Partition, Int Selection Array=DataInterface() {propertyValues {IntData [0,2,3,6,4,7,5,8]}})
    Bone Sampling Mode=::NewEnumerator4
    MaskSpawnGroup=false
    Sampled Position Offset=&Multiply_VectorByFloat(
     Vector=&MergeHelper_VectorSwitch(
      0=&SelectVectorFromArray(
       Array Sampling Mode=::NewEnumerator0
       Direct Array Index=?
       Vector Selection Array=DataInterface() {propertyValues {FloatData [
          <13.8012 2.8733 -10.9414>
          <10.6497 8.5547 1.6815>
          <13.3621 5.3119 -9.0131>
          <9.5659 7.9116 -4.7063>
          <9.1597 8.6339 -2.8767>
          <9.6117 8.6012 -2.8833>
          <8.6766 8.8321 -1.9797>
          <8.0164 9.0567 -0.2421>
          <5.235 9.1319 -0.3265>
          <8.0172 8.9594 1.341>
          <13.0733 7.5045 5.3352>
          <13.5945 6.1913 7.9272>
          <14.2142 6.9145 6.262>
          <13.6977 6.1902 7.9121>
         ]}}
       Random Seed=0
      )
      1=&SelectVectorFromArray(
       Array Sampling Mode=::NewEnumerator0
       Direct Array Index=?
       Vector Selection Array=DataInterface() {propertyValues {FloatData [
          <2.1101 1.3498 12.7021>
          <1.9596 3.0538 12.2868>
          <4.2619 5.2207 11.6814>
          <1.8981 5.7386 11.4301>
          <-9.4265 1.6418 11.7511>
          <5.7517 8.6447 8.5826>
          <10.9287 7.3928 3.1786>
          <11.9879 7.183 0.0035>
          <10.0883 7.9329 -3.0182>
          <8.1823 7.5211 -7.7231>
          <10.84 7.6668 -2.1509>
          <11.1784 7.0818 -3.7262>
          <10.4436 7.1121 -4.9948>
          <2.0237 4.3374 -11.9317>
          <-9.9873 2.7377 -11.609>
          <-10.5202 5.6944 -10.25>
          <-11.8073 5.5858 -10.3405>
          <-12.0672 7.1213 9.0279>
          <-12.9381 5.3014 8.5794>
          <1.7399 -0.0275 -12.9438>
          <0.7374 -2.0823 -12.808>
          <0.8108 0.0426 -12.791>
          <1.8457 -5.4161 -12.1659>
          <0.0554 -0.1432 12.6648>
          <0.2765 -1.841 12.7144>
          <1.1496 -3.4887 12.7097>
          <-8.8196 -4.8406 10.4307>
          <-7.8299 -5.2169 10.056>
          <-10.6478 -0.6625 -12.1933>
          <-10.5372 -1.2185 -12.1841>
          <-8.8704 -0.7906 -12.127>
          <-7.9667 -4.4983 -10.8169>
          <-10.0035 -7.5704 -7.1045>
          <-8.2389 -5.9284 -9.5964>
          <-8.6197 -8.2115 -5.6535>
          <-7.8846 -7.3589 -7.4054>
          <-9.2236 -9.1277 -0.7388>
          <-7.3657 -8.7301 2.4598>
          <-8.7915 -8.5869 4.1153>
          <-7.5249 -5.8982 9.4217>
          <2.4682 -5.9704 11.9725>
          <-12.1097 6.2673 -9.2839>
         ]}}
       Random Seed=0
      )
      2=&SelectVectorFromArray(
       Array Sampling Mode=::NewEnumerator0
       Direct Array Index=?
       Vector Selection Array=DataInterface() {propertyValues {FloatData [
          <18.7053 14.7296 -5.0338>
          <17.4119 14.5086 -4.9037>
          <16.7615 14.3857 -5.1024>
          <18.1661 13.7729 -0.1349>
          <17.1913 12.3104 2.8906>
          <18.7344 10.0042 6.8968>
          <18.8046 9.3249 7.5794>
          <18.3442 8.2928 8.3195>
          <16.3676 9.1258 6.8829>
          <17.4869 8.1517 8.0789>
          <15.8404 8.2532 7.521>
          <18.3035 6.2531 9.6156>
          <16.8103 3.8439 10.0736>
          <15.5586 6.7156 8.5052>
          <16.2393 3.6691 9.9798>
          <14.951 14.3858 -8.7569>
          <15.0987 14.2773 -6.6643>
          <15.5639 13.9813 -3.6821>
          <15.3121 13.1402 -0.1084>
          <15.0042 12.9296 0.4626>
          <15.701 11.7778 3.2555>
          <15.1753 11.113 4.3562>
          <18.9705 -9.2853 -9.2962>
          <15.8912 -8.9419 -7.8383>
          <18.9328 -9.6315 -4.489>
          <17.9364 -9.4208 -5.3175>
          <18.918 -9.645 -4.4066>
          <19.2399 -9.5644 -0.8259>
          <15.6601 -8.9838 -5.3056>
          <19.1168 -9.666 -0.0086>
          <17.933 -9.5644 -2.3089>
          <16.0545 -9.1546 -1.8179>
          <15.407 -9.0096 -3.5991>
          <19.0226 -8.5979 4.4393>
          <18.9031 -9.3041 2.8015>
          <17.7156 -8.9814 2.5524>
          <17.0753 -8.1188 4.32>
          <16.5039 -3.0431 9.0125>
          <16.9023 -1.1739 9.8483>
          <16.1379 -0.6089 9.8358>
          <15.1694 0.2598 9.8672>
          <15.0392 -3.7371 8.1621>
          <15.9412 -4.8158 7.5915>
          <16.261 -7.4229 4.8994>
          <14.9906 -7.4992 4.295>
          <15.1298 -8.1428 3.1213>
          <15.974 -8.8997 0.776>
          <14.9036 -8.8797 -1.3688>
          <14.5777 -8.5152 -1.5754>
         ]}}
       Random Seed=0
      )
      3=&SelectVectorFromArray(
       Array Sampling Mode=::NewEnumerator0
       Direct Array Index=?
       Vector Selection Array=DataInterface() {propertyValues {FloatData [
          <-14.3494 -4.6048 -9.257>
          <-15.8534 -5.8474 -9.3021>
          <-16.483 -6.4736 -9.1221>
          <-17.8841 -10.7686 -6.1051>
          <-17.9249 -12.0136 -4.2779>
          <-14.6787 -11.0161 -4.4329>
          <-17.0409 -11.842 -3.9745>
          <-17.8602 -13.5377 -0.8521>
          <-16.1265 -12.6827 -1.857>
          <-17.4969 -13.9594 0.975>
          <-15.3314 -13.2195 0.0789>
          <-16.7737 -14.2279 3.5155>
          <-16.348 -14.5325 8.2409>
          <-15.5852 -14.3995 6.932>
          <-15.6517 -14.4512 8.2066>
          <-16.2458 0.0716 -10.2191>
          <-16.903 2.3159 -9.5914>
          <-17.1464 6.0937 -7.0079>
          <-18.6686 9.0142 -3.0295>
          <-17.8831 9.1186 -1.7471>
          <-16.3006 8.9198 -0.6697>
          <-18.2904 9.5348 1.0994>
          <-16.1187 8.9718 -0.3258>
          <-18.8904 9.5944 4.6522>
          <-15.1096 8.7872 7.1345>
         ]}}
       Random Seed=0
      )
      4=&SelectVectorFromArray(
       Array Sampling Mode=::NewEnumerator0
       Direct Array Index=?
       Vector Selection Array=DataInterface() {propertyValues {FloatData [
          <33.0731 7.5506 0.7231>
          <29.7452 7.6631 0.6326>
          <26.2292 8.0939 0.012>
          <20.135 9.1532 0.2868>
          <28.0947 5.7015 -2.737>
          <29.053 3.7759 -3.0437>
          <29.5431 3.0671 -3.1061>
          <33.1731 2.0331 -2.4908>
          <35.1531 5.3481 -1.7869>
          <33.7886 1.4966 -2.246>
          <35.611 -0.4619 -0.682>
          <36.6516 4.7087 4.1971>
          <36.4074 6.9535 1.68>
          <36.1073 6.3541 3.0004>
          <31.0996 7.258 2.3636>
          <35.0702 1.9975 4.5571>
          <34.2393 4.2185 4.5957>
          <32.0921 4.2398 4.6575>
          <32.8597 -0.3809 3.035>
          <31.4508 2.7528 4.8927>
          <30.4655 2.2682 5.0189>
          <30.8087 5.0953 4.5383>
          <30.5251 6.4306 3.7202>
          <28.8323 7.4712 2.6488>
          <25.2146 8.4173 1.8118>
          <25.5012 6.5492 4.6078>
          <22.9148 8.7424 0.6839>
          <15.2401 9.4075 1.2897>
          <19.9816 8.8786 2.7878>
          <21.9822 6.4671 5.5143>
          <11.3419 8.8414 2.355>
          <12.5128 3.0817 6.6728>
          <12.0181 4.0861 6.5326>
          <10.8853 8.3244 3.5719>
          <10.0513 0.7842 5.7365>
          <10.1367 -0.2611 4.8115>
          <10.7784 -2.0879 -0.8239>
          <10.5092 0.2958 -4.0702>
          <10.0816 3.8075 -4.8018>
          <12.4391 5.4014 -4.764>
          <15.1808 -0.192 -3.6269>
          <14.6179 4.583 -4.7875>
          <16.6672 -0.9792 -2.7017>
          <15.98 -2.6678 -0.408>
          <19.0657 -0.277 -3.0589>
          <17.855 0.693 -3.8028>
          <11.9237 -2.8661 2.9882>
          <14.6981 -3.2093 1.5513>
          <21.9636 -0.8273 -2.0461>
          <13.5304 6.1088 6.2605>
          <13.8143 6.812 5.835>
          <14.19 8.2272 4.4871>
          <14.2386 8.5532 3.9907>
          <18.399 -2.7031 0.9248>
          <23.1318 -1.3757 3.4663>
          <25.1515 1.9947 5.6674>
          <29.1156 -1.3107 0.5526>
          <29.4933 -0.9371 -0.6685>
          <27.5522 -0.9541 -1.0392>
          <29.1761 -0.3703 -1.317>
          <29.5861 2.3564 -2.9101>
          <28.4339 1.7211 -2.8354>
          <25.8009 4.6223 -3.4143>
          <24.5462 4.5573 -3.6806>
          <18.2773 2.5683 -4.6003>
          <21.9723 5.1496 -4.0772>
          <18.355 3.0671 -4.5859>
          <17.089 4.5411 -4.6621>
          <20.8448 6.5695 -3.5716>
          <20.0404 6.9537 -3.3922>
          <17.0247 7.9591 -2.8425>
          <16.4913 8.5682 -2.1979>
          <13.3959 8.3353 -2.6102>
          <13.3425 5.774 -4.5016>
          <12.9252 8.9402 -1.1481>
          <9.324 7.9691 -0.8149>
         ]}}
       Random Seed=0
      )
      5=&SelectVectorFromArray(
       Array Sampling Mode=::NewEnumerator0
       Direct Array Index=?
       Vector Selection Array=DataInterface() {propertyValues {FloatData [
          <-30.8056 -7.459 -0.1441>
          <-30.2873 -7.515 0.0282>
          <-29.4345 -7.5049 -0.0008>
          <-32.7424 -5.1241 2.3974>
          <-32.8908 -3.0442 2.7988>
          <-35.2062 -4.2983 2.2671>
          <-33.8773 1.0606 -1.268>
          <-34.0296 0.7605 0.5947>
          <-31.8279 1.0791 -1.033>
          <-30.3448 0.6209 0.9106>
          <-30.3532 1.1324 -0.4486>
          <-28.0811 -5.1054 3.0286>
          <-27.0896 -2.6251 3.3228>
          <-27.6264 -3.6195 3.3212>
          <-25.4173 -7.8164 1.1105>
          <-25.5042 -7.6673 1.3659>
          <-24.7808 -6.6127 2.7336>
          <-21.865 -8.4677 1.2169>
          <-20.5399 -7.3519 2.9899>
          <-18.3502 -8.45 2.1675>
          <-19.5256 -6.1954 4.0366>
          <-21.4575 -4.2982 4.1881>
          <-17.6363 -4.127 4.6193>
          <-19.7343 -4.9033 4.3855>
          <-22.9649 -0.6419 3.0978>
          <-27.8086 0.7101 1.2126>
          <-21.3831 -0.0533 3.0043>
          <-27.1772 1.6772 -1.0069>
          <-23.8667 1.6775 0.4552>
          <-22.6663 1.8249 0.4646>
          <-26.4004 1.0258 -3.238>
          <-19.8154 1.8876 1.2559>
          <-21.3594 2.3884 -1.6467>
          <-18.1853 -0.0816 3.4849>
          <-17.2854 2.92 -1.2379>
          <-20.0518 2.4242 -0.6305>
          <-16.6079 -0.2562 3.7339>
          <-16.7236 -4.646 4.6949>
          <-16.4357 1.1954 2.5021>
          <-15.6372 -5.496 4.7122>
          <-16.2143 -7.1982 3.4689>
          <-15.6106 1.2334 2.5763>
          <-15.3318 2.5664 1.1089>
          <-12.3866 -8.9894 0.9685>
          <-10.1835 -8.7378 0.9408>
          <-10.1274 -8.2029 2.2532>
          <-11.8125 -7.1502 3.5602>
          <-13.134 -4.3395 4.8601>
          <-10.595 -2.3343 3.8624>
          <-10.2957 -2.1272 4.7182>
          <-13.2499 0.504 3.6538>
          <-12.8182 1.6559 2.6568>
          <-10.3198 0.8031 -5.0843>
          <-10.8959 1.8149 -4.4955>
          <-10.8998 0.0325 -5.673>
          <-10.0892 -5.2541 -6.1825>
          <-10.2369 -8.4521 -2.9359>
          <-12.2151 -9.0288 -2.2307>
          <-10.5154 -7.9738 -4.0852>
          <-28.7811 -7.7504 -1.3308>
          <-35.0471 -6.1655 -3.373>
          <-32.6786 -4.6474 -4.5608>
          <-32.4496 -0.8687 -4.3888>
          <-29.8207 0.9606 -2.5174>
          <-28.4 0.9706 -2.8082>
          <-29.2873 -0.6832 -4.4697>
          <-30.0497 -2.1402 -5.0186>
          <-28.4678 -5.5469 -4.873>
          <-27.1419 -1.1475 -5.1449>
          <-26.9172 0.26 -4.0518>
          <-25.4821 -0.0821 -4.6006>
          <-24.398 0.0342 -4.6711>
          <-24.5041 -7.4487 -3.893>
          <-22.9245 -5.0192 -6.1348>
          <-21.7266 -7.976 -3.9233>
          <-23.1316 -0.2206 -5.1476>
          <-19.5376 1.9158 -3.3992>
          <-11.9143 3.2844 -1.9979>
          <-12.4994 -1.2472 -6.1914>
          <-13.8551 -1.5662 -6.3604>
          <-15.646 -2.9579 -6.913>
          <-16.1968 -9.1231 -2.8494>
          <-13.1992 -8.5348 -3.8929>
          <-13.2585 -7.9988 -4.5444>
          <-13.2651 -3.9712 -6.6099>
         ]}}
       Random Seed=0
      )
      6=&SelectVectorFromArray(
       Array Sampling Mode=::NewEnumerator0
       Direct Array Index=?
       Vector Selection Array=DataInterface() {propertyValues {FloatData [
          <-2.216 5.7949 1.8138>
          <-2.718 5.1642 2.6089>
          <-1.4811 0.6478 4.2938>
          <0.4765 -1.7392 4.1657>
          <3.5983 -4.3189 4.4638>
          <4.6057 -5.1737 4.6855>
          <2.5295 -14.1524 2.2346>
          <3.1738 -13.4465 3.5449>
          <2.6592 -11.939 -3.5313>
          <4.0025 -7.9026 -3.8659>
          <3.3873 -5.4431 -3.5258>
          <4.6498 -6.5608 -3.7258>
          <5.9037 -8.281 -3.98>
          <6.9169 -12.1956 -3.7315>
          <7.2692 -12.9706 -3.4634>
          <5.5877 -14.7749 1.4238>
          <4.5125 -13.6131 3.6225>
          <3.9138 -12.7533 4.0748>
          <-6.0381 4.7089 2.4815>
          <-7.759 4.1813 2.2891>
          <-4.277 -0.1844 4.2215>
          <-1.2293 -2.3015 4.4027>
          <0.01 -3.1821 4.2896>
          <0.0399 -5.0193 4.4026>
          <0.3081 -7.3143 4.322>
          <0.3021 -7.7082 4.2886>
          <1.597 -10.9509 4.3224>
          <2.2158 -7.5318 -3.4897>
          <-2.2871 -2.3377 -3.1584>
          <-3.2164 3.7284 -2.4872>
          <-4.3394 0.8973 -3.3131>
          <-7.1254 4.422 -1.7824>
          <-9.8301 0.9051 -3.032>
          <-7.1545 -0.4307 -3.1259>
          <-6.3749 -2.7095 -2.9044>
          <-7.1853 -3.081 -2.4978>
          <-4.8548 -3.5614 -2.8714>
          <-2.7871 -4.6615 -2.8861>
          <-2.6791 -6.1751 -2.4437>
          <-3.898 -6.6314 -1.6276>
          <-0.4005 -9.4319 -2.9433>
          <-1.6093 -10.1009 -1.3281>
          <-0.7877 -12.9177 1.5419>
          <-1.2658 -12.4785 1.2142>
          <-1.8984 -9.3282 1.2257>
          <-1.3321 -8.8965 2.8551>
          <-1.6262 -6.3361 3.695>
          <-4.1644 -6.2311 2.0234>
          <-6.6824 -4.4828 2.3132>
          <-7.0501 -4.3501 2.2426>
          <-6.1842 -0.4462 3.8664>
          <-9.3233 -0.593 3.4946>
          <-7.2626 2.3912 3.4966>
          <0.063 0.5837 -3.2562>
          <-2.2738 3.2135 -2.884>
          <-1.6993 5.9374 -0.4104>
          <3.5499 5.2271 -1.6255>
          <3.3137 3.5745 -2.865>
          <3.5508 2.1684 -3.238>
          <3.6693 0.8178 -3.2889>
          <6.6931 2.5543 -3.1342>
          <7.9411 2.7557 4.1554>
          <4.5882 3.6154 4.0385>
          <2.46 5.0939 3.034>
          <7.9345 4.8157 3.1111>
          <7.365 5.0521 3.0648>
          <7.5423 5.6962 2.2045>
         ]}}
       Random Seed=0
      )
      7=&SelectVectorFromArray(
       Array Sampling Mode=::NewEnumerator0
       Direct Array Index=?
       Vector Selection Array=DataInterface() {propertyValues {FloatData [
          <0.0997 -5.9458 -1.7334>
          <-0.7148 -6.062 -1.2376>
          <2.5689 -5.3524 -2.4294>
          <2.6294 -4.7203 -3.0735>
          <2.4084 -5.368 -2.4351>
          <-6.1623 5.1315 -4.6955>
          <-2.9361 5.1395 -4.5831>
          <-2.729 8.1912 -4.7399>
          <-2.5903 9.5273 -4.6134>
          <-2.9339 13.3817 2.7182>
          <-3.9451 13.643 2.5734>
          <-4.7446 7.1177 3.7977>
          <-7.1433 7.9312 3.9908>
          <-5.1904 11.4834 3.7903>
          <-6.6951 12.4085 3.6819>
          <-6.5076 12.514 3.6166>
          <-5.4127 13.8641 2.5855>
          <-4.2331 12.5323 -4.2338>
          <-6.7376 13.032 -4.1976>
          <-6.2651 11.1041 -4.7933>
          <-7.1242 11.4803 -4.7605>
          <-7.2043 6.8179 -4.9526>
          <4.0824 -5.3711 -1.9652>
          <-1.2081 4.2442 -4.2853>
          <-1.3927 12.9767 -3.4329>
          <3.4384 2.7686 -4.1577>
          <5.9533 4.6315 -2.5351>
          <5.4893 3.4554 -3.3972>
          <6.06 2.9114 2.8544>
          <7.9938 4.6959 0.2305>
          <1.1132 11.7505 1.5727>
          <-0.0956 11.8336 2.7185>
          <-1.3283 6.0774 3.336>
          <0.3263 3.715 3.1474>
          <0.5057 1.9324 2.8908>
          <1.0047 1.0141 2.8682>
          <3.2113 1.5946 3.2343>
          <2.8781 -1.4419 3.1105>
          <4.5614 -1.1769 3.2723>
          <5.4412 -5.6321 0.5297>
          <7.6547 -3.6923 2.1113>
          <8.7 -2.7062 2.4582>
          <10.5645 -3.2724 1.5752>
          <7.828 0.009 3.1891>
          <9.0092 1.1486 3.0292>
          <9.9922 4.232 -0.2581>
          <9.1737 3.887 -1.5704>
          <8.4102 2.5747 -2.9139>
          <7.5622 0.7596 -3.5814>
          <7.646 -1.1413 -3.6039>
          <1.4641 -6.1164 -0.3466>
          <0.2158 -6.0626 0.0774>
          <0.6461 -5.5009 1.3539>
          <-2.1761 -6.1086 -0.1549>
          <-2.4717 -5.7679 0.9791>
          <-2.781 -3.7138 2.7926>
          <-6.9414 -5.4301 1.2591>
          <-8.1044 -5.785 0.0085>
          <-1.2154 -5.9616 -1.7001>
          <-1.0402 -5.7322 -2.2893>
          <-6.7497 -5.8967 -1.6533>
          <-3.3296 -1.125 -4.5128>
          <-5.4355 -3.2568 -4.1394>
          <-6.5171 -2.3859 -4.3274>
          <-8.1635 -5.498 -1.4946>
         ]}}
       Random Seed=0
      )
      8=&SelectVectorFromArray(
       Array Sampling Mode=::NewEnumerator0
       Direct Array Index=?
       Vector Selection Array=DataInterface() {propertyValues {FloatData []}}
       Random Seed=0
      )
      Index=$Particles.PartitionParticles.Partition
      Num=::NewEnumerator6
     )
     Float=0.95
    )
    Skeletal Mesh=$System.Skeletal Mesh
   )
   &ScaleSpriteSize(
    Uniform Curve Scale=&FNTimeOfDayScalar(Night Scale=1.25, Morning Scale=1, Evening Scale=1)
    Uniform Curve Sprite Scale=DataInterface() {propertyValues {Curve {keys [
        {interpMode=RCIM_Cubic, tangentMode=RCTM_User, time=0, value=0}
        {interpMode=RCIM_Cubic, tangentMode=RCTM_User, time=0.5, value=1}
        {tangentMode=RCTM_User, time=1, value=0}
       ]}}}
   ) {module=ScaleSpriteSize001}
   &ScaleColor(
    Scale Alpha=&Multiply_Float(
     A=&Multiply_Float(
      A=&FloatFromCurve(
       FloatCurve=DataInterface() {propertyValues {Curve {keys [
           {arriveTangent=6.5707, interpMode=RCIM_Cubic, leaveTangent=6.5707, tangentMode=RCTM_User, time=0, value=0}
           {interpMode=RCIM_Cubic, tangentMode=RCTM_User, time=0.3, value=1}
           {tangentMode=RCTM_User, time=1, value=0}
          ]}}}
       Scale Curve=&RandomRangeFloat(Minimum=0.1, Maximum=0.5)
      )
      B=&MakeFloatFromLinearColor(LinearColor=$System.TODColorMultiplier, Channel:ENiagaraLinearColor_Channels=::NewEnumerator3)
     )
     B=&FN_FrontEndFloatModifier(
      Gameplay=1
      Front End=&MakeCustomFloatFromBool(
       Bool=&SetBoolByFloatComparison(A=$NPC.FortniteNPC.EventAudio_Amplitude, B=0)
       True Float=1
       False Float=&FloatFromCurve(
        FloatCurve=DataInterface() {propertyValues {Curve {keys [<at 0 0 RCIM_Cubic>, {arriveTangent=1.362, interpMode=RCIM_Cubic, leaveTangent=-0.9886, tangentMode=RCTM_Break, time=0.5, value=1}, <at 1 0>]}}}
        CurveIndex=&Cosine(
         Normalized Angle=&Add_Float(A=&Divide_Float(A=$Engine.Time, B=5), B=0.5)
         Period=1
         Scale=0.5
         Bias=1
        )
       )
      )
     )
    )
    Scale RGB=&Multiply_Vector(
     A=&VectorFromFloat(Value=&Multiply_Float(
       A=&RandomRangeFloat(Minimum=0.75, Maximum=1)
       B=&SelectFloatFromArray(Direct Array Index=$Engine.QualityLevel, Float Selection Array=DataInterface() {propertyValues {FloatData [10,5,1,1,1]}})
      ))
     B=&MakeVectorFromLinearColorRGB(Color=$System.TODColorMultiplier)
    )
    ScaleA=true
    ScaleRGB=true
   ) {module=ScaleColor001}
   /SetVariables(
    Particles.CenterFlareColor:LinearColor=&MultiplyLinearColor(
     Linear Color=$Particles.Color
     Scale Factor=&MakeVector4_FromLinearColor(LinearColor=&LerpLinearColors(StartColor=<rgba 10 10 10 1>, EndColor=<rgba 10 10 10 1>, LerpFactor=$User.Variant))
    )
    Particles.CenterFlareSize:Vector2f=&Multiply_Vector2(A=$Particles.SpriteSize, B=&Vector2DFromFloat(Value=0.5))
    Particles.CoronaFlareColor:LinearColor=&MultiplyLinearColor(
     Linear Color=$Particles.Color
     Scale Factor=&MakeVector4_FromLinearColor(LinearColor=&LerpLinearColors(StartColor=<rgba 0.15 0.15 0.15 1>, EndColor=<rgba 0.0729 0.0729 0.0729 1>, LerpFactor=$User.Variant))
    )
    Particles.CoronaFlareSize:Vector2f=&Multiply_Vector2(A=$Particles.Initial.SpriteSize, B=&Vector2DFromFloat(Value=0.75))
   ) {isSetParameters=true, module=SetVariables_#9}
  ]
  rendererClasses [&NiagaraSpriteRendererProperties]
  renderers [
   {rendererClass=&NiagaraSpriteRendererProperties}
   {rendererClass=&NiagaraSpriteRendererProperties, rendererIndex=1}
  ]
 }
}
summary {
 systemName=NS_Character_DimeBlanketGrace_Idle
 emitters [
  {emitterName=Hearts, rendererClasses [&NiagaraSpriteRendererProperties]}
  {emitterName=PixieDust, rendererClasses [&NiagaraSpriteRendererProperties]}
  {emitterName=Sparkles_Pink, rendererClasses [&NiagaraSpriteRendererProperties]}
  {emitterName=Sparkles_Silver, rendererClasses [&NiagaraSpriteRendererProperties]}
 ]
 userVariables [
  User.bIsADS {type=NiagaraBool}
  User.bisFootHidden {type=NiagaraBool}
  User.bIsFrontEnd {type=NiagaraBool}
  User.bIsFrontEndPreview {type=NiagaraBool}
  User.MusicReactive {type=NiagaraBool}
  User.Variant {type=NiagaraFloat, defaultValue=0, description=""}
 ]
}
