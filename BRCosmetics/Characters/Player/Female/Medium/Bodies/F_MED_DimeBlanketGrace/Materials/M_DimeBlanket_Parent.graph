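# Parent character material for the DimeBlanket Grace outfit, adding sequins and audio-reactive glow.
# Between the CharacterParent 2023 PreFX and PostFX functions, it chains two MF_Sequins layers (secondary optional) with tech-art channel masks and adds switchable emissive.
# The glow combines panned caustics, Fresnel height-lerp, tint and audio amplitude, with a debug sine, time-based FEAudio fallback and optional crop mask.
# A metal reflectance LUT trim branch appears unconnected.
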Material /BRCosmetics/Characters/Player/Female/Medium/Bodies/F_MED_DimeBlanketGrace/Materials/M_DimeBlanket_Parent
graph {
 properties {
  bAutomaticallySetUsageInEditor=false
  bUseMaterialAttributes=true
  bUsedWithMeshDeformer=true
  bUsedWithMorphTargets=true
  bUsedWithSkeletalMesh=true
  shadingModel=MSM_FromMaterialExpression
  usedShadingModels=Default Lit | Subsurface | Clear Coat | Subsurface Profile | Cloth
 }
 nodes [
  n1 = ScalarParameter() {defaultValue=10, parameterName=FE_AudioReactivityFrequency}
  n2 = MaterialFunctionCall(Speed=@n1) {materialFunction=&DebugTimeSine}
  Add(
   A=Multiply(A=MaterialFunctionCall() {materialFunction=&MF_Hash}.Float) {constB=0.05}
   B=ConstantBiasScale(None=@n2.Result) {scale=0.3}
  )
  n3 = MaterialFunctionCall() {materialFunction=&MF_CharacterParent_2023_PreFX}
  n4 = TextureObjectParameter() {
   group=Sequins_Main
   parameterName=SequinRoughness_Main
   samplerType=SAMPLERTYPE_LinearGrayscale
   texture=/Game/Global/Textures/Default/Tileable_Patterns/T_SequinTile_roughness
  }
  n5 = TextureObjectParameter() {
   group=Sequins_Main
   parameterName=SequinOffest_Main
   samplerType=SAMPLERTYPE_LinearGrayscale
   texture=/Game/Global/Textures/Default/Tileable_Patterns/T_SequinTile
  }
  n6 = TextureObjectParameter() {
   group=Sequins_Main
   parameterName="SequinNormal_,Main"
   samplerType=SAMPLERTYPE_Normal
   texture=/Game/Global/Textures/Default/Tileable_Patterns/T_SequinTile_N
  }
  n7 = VectorParameter() {group=Sequins_Main, parameterName=SequinFalloffColor02_Main, defaultValue=<rgba 0.8438 0.4263 0.6791 1>}
  n8 = TextureObjectParameter() {
   group=Sequins_Secondary
   parameterName=SequinThinFilm_Trim
   texture=/Game/Global/Textures/Default/Thinfilm/T_ThinFilm_Spectrum_BlueGreen_Color
  }
  n9 = MaterialFunctionCall!(
   inMat=@n3.Result
   SequinRoughness=@n4
   SequinOffset=@n5
   SequinNormal=@n6
   SequinFalloffColor01=VectorParameter() {group=Sequins_Main, parameterName=SequinFalloffColor01_Main, defaultValue=<rgba 0.8542 0.387 0.5588 1>}.RGB
   SequinFalloffColor02=@n7.RGB
   SequinRotationAngle=ScalarParameter() {group=Sequins_Main, parameterName=SequinRotationAngle_Main}
   SequinThinFilmTexture=@n8
   TechArtMask=NamedRerouteUsage()
  ) {materialFunction=/Game/Global/MaterialFunctions/TechArt/MF_Sequins}
  n10 = TextureObjectParameter() {
   group=Sequins_Secondary
   parameterName=SequinRoughness_Secondary
   samplerType=SAMPLERTYPE_LinearGrayscale
   texture=&T_SequinTile_roughness
  }
  n11 = TextureObjectParameter() {
   group=Sequins_Secondary
   parameterName=SequinOffest_Secondary
   samplerType=SAMPLERTYPE_LinearGrayscale
   texture=&T_SequinTile
  }
  n12 = TextureObjectParameter() {
   group=Sequins_Secondary
   parameterName="SequinNormal_,Secondary"
   samplerType=SAMPLERTYPE_Normal
   texture=&T_SequinTile_N
  }
  n13 = VectorParameter() {group=Sequins_Secondary, parameterName=SequinFalloffColor01_Secondary, defaultValue=<rgba 0.6777 0.6151 0.9062 1>}
  n14 = VectorParameter() {group=Sequins_Secondary, parameterName=SequinFalloffColor02_Secondary, defaultValue=<rgba 0.6332 0.6965 1 1>}
  n15 = TextureObjectParameter() {
   group=Sequins_Secondary
   parameterName=SequinThinFilm_Secondary
   texture=&T_ThinFilm_Spectrum_BlueGreen_Color
  }
  n16 = MaterialFunctionCall(
   inMat=@n9.Result
   SequinRoughness=@n10
   SequinOffset=@n11
   SequinNormal=@n12
   SequinFalloffColor01=@n13.RGB
   SequinFalloffColor02=@n14.RGB
   SequinRotationAngle=ScalarParameter() {group=Sequins_Secondary, parameterName=SequinRotationAngle_Secondary}
   SequinThinFilmTexture=@n15
   TechArtMask=NamedRerouteUsage()
  ) {materialFunction=&MF_Sequins}
  n17 = GetMaterialAttributes(None=StaticSwitchParameter(True=@n16.Result, False=@n9.Result) {defaultValue=true, parameterName=UseSecondarySequin}) {attributeGetTypes [#1]}
  n18 = StaticSwitchParameter(True=NamedRerouteUsage(), False=Constant()) {defaultValue=true, parameterName=DimeBlanket_UseAudioReactivity}
  MaterialFunctionCall(
   Material=SetMaterialAttributes(MaterialAttributes=@n17.MaterialAttributes, Emissive Color=Add(A=@n17.Emissive Color, B=@n18)) {attributeSetTypes [#1]}
   M=@n3.M
   Pre Skinned Local Position=@n3.M
   Distance Blend=@n3.M
   Fresnel=@n3.M
  ) {materialFunction=&MF_CharacterParent_2023_PostFX}
  Multiply()
  n19 = TextureCoordinate()
  n20 = ScalarParameter() {defaultValue=2, parameterName=AudioReactivity_NoiseTiling}
  n21 = TextureSample(UVs=Panner(Coordinate=Multiply(A=@n19, B=@n20)) {speedY=0.1}) {samplerSource=SSM_Wrap_WorldGroupSettings, texture=/Game/Effects/Fort_Effects/Textures/Tile/T_ShieldCaustics_Tile_Chromatic}
  n22 = MaterialFunctionCall(
   Transition Phase=MaterialFunctionCall(Normal Vector=PixelNormalWS(), Power=ScalarParameter() {defaultValue=1, parameterName=AudioReactivity_FresnelPower}) {materialFunction=&Fresnel_Function}.Result
   Height Texture=ShadingPathSwitch(Default=@n21.R, Deferred=@n21.R, Mobile=Constant()) {desc=Migrated from FeatureLevelSwitch}
  ) {materialFunction=&HeightLerp}
  Multiply(A=@n22.Alpha)
  NamedRerouteDeclaration(None=ChannelMaskParameter(None=@n3.Tech Art Mask) {group=Sequins_Secondary, parameterName=Sequin_SecondaryChannel}.R) {name=Sequin_Secondary, nodeColor=<rgba 1 0.75 0 1>}
  n23 = ChannelMaskParameter(None=@n3.Tech Art Mask) {group=Sequins_Main, maskChannel=Green, parameterName=Sequin_TrimChannel, defaultValue=<rgba 0 1 0 0>}
  NamedRerouteDeclaration(None=@n23.R) {name=Sequin_Trim, nodeColor=<rgba 1 0.75 0 1>}
  NamedRerouteDeclaration(None=ChannelMaskParameter(None=@n3.Tech Art Mask) {parameterName=AudioReactivityChannel}.R) {name=AudioReactivityMask, nodeColor=<rgba 1 0 0.0937 1>}
  n24 = StaticSwitchParameter(
   True=@n2.Result
   False=MaterialFunctionCall() {materialFunction=/Game/Packages/Fortress_SharedMaterials/Base_Material_Functions/BaseCharacterShaderFunctions/Audio/MF_AudioReactivity}.Amplitude
  ) {parameterName=AudioDebug}
  n25 = ScalarParameter() {defaultValue=3, parameterName=AmplitudePower}
  n26 = Divide(A=Time(), B=@n1)
  n27 = ScalarParameter() {defaultValue=0.75, parameterName=FE_AudioReactivityOffset}
  n28 = Abs(None=Cosine(None=Add(A=@n26, B=@n27)))
  NamedRerouteDeclaration(None=LinearInterpolate(A=Power(Base=@n24, Exp=@n25) {constExponent=3}, B=OneMinus(None=@n28), Alpha=ScalarParameter() {parameterName=FEAudio})) {name=AudioInput, nodeColor=<rgba 1 0.3047 0 1>}
  n29 = SphereMask(A=TextureCoordinate(), B=Constant2Vector() {g=0.068, r=0.285}) {attenuationRadius=0.04, hardnessPercent=0}
  n30 = Multiply(A=OneMinus(None=@n29), B=OneMinus(None=Step(Y=@n29) {constX=0}))
  n31 = StaticSwitchParameter(True=Add(A=@n22.Alpha, B=Power(Base=@n30) {constExponent=0.8}), False=@n22.Alpha) {group=DimeBlanket, parameterName=IsDimeBlanketGraceBody}
  n32 = VectorParameter() {group=Sequins_Secondary, parameterName=AudioReactivityColor, defaultValue=<rgba 3 0.2 1 1>}
  n33 = Multiply(A=Multiply(A=Multiply(A=@n31, B=@n32.RGB), B=NamedRerouteUsage()), B=NamedRerouteUsage())
  n34 = VectorParameter() {parameterName=2, defaultValue=<rgba 0.9 0.4 1 1>}
  n35 = LinearInterpolate(
   A=@n33
   Alpha=MaterialFunctionCall!(
    UVs=TextureCoordinate()
    TextureIn
    UpperLeftCorner=AppendVector(A=@n34.R, B=@n34.R)
    LowerRightCorner=AppendVector(A=@n34.B, B=@n34.B)
    ExportFloat4
   ) {materialFunction=/Engine/Functions/Engine_MaterialFunctions01/Texturing/TextureCropping}.Crop Mask
  ) {constB=0}
  NamedRerouteDeclaration(None=Multiply(
    A=StaticSwitchParameter(True=@n35, False=@n33) {parameterName=HidingCrouchReactivity}
    B=ScalarParameter() {defaultValue=1, parameterName=AudioReactivity_FinalStrength}
   )) {name=AudioReactivity, nodeColor=<rgba 0.7813 0 1 1>}
  n36 = GetMaterialAttributes() {attributeGetTypes [#2, #3]}
  n37 = CameraVectorWS()
  n38 = PixelNormalWS()
  n39 = CurveAtlasRowParameter(CurveTime=Clamp(None=DotProduct(A=@n37, B=@n38))) {
   atlas=&AthenaCurveAtlas01
   curve=/Game/Characters/Player/Female/Medium/Bodies/F_MED_Ms_Alpine/Textures/AluminiumreflectanceCurve
   defaultValue=189
   group=LUT
   parameterName=FresnelReflectance_LUT
  }
  n40 = Multiply(A=@n36.Base Color, B=ChannelMaskParameter(None=@n39) {group=LUT, parameterName=LUTChannel}.R)
  n41 = NamedRerouteUsage()
  StaticSwitchParameter(
   True=SetMaterialAttributes!(MaterialAttributes=@n36.MaterialAttributes, Base Color=LinearInterpolate(A=@n36.Base Color, B=@n40, Alpha=@n41)) {attributeSetTypes [#3]}
   False=@n16.Result
  ) {defaultValue=true, parameterName=DimeBlanket_UseMetalLUTforTrim}
  Transform(None=@n36.Normal)
 ]
}
