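# Physics asset for the DimeBlanketGrace female medium body cosmetic, defining bodies and constraints for skirt physics.
# Six skirt chains each pair a tiny base capsule with a larger dyn capsule; middle chains attach to the pelvis, side chains to the matching thigh.
# Pelvis and thigh capsules provide the body volumes.
# Base and thigh joints swing and twist freely, while tight dyn-to-base limits allow mostly single-axis, hinge-like flap motion.
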PhysicsAsset /BRCosmetics/Characters/Player/Female/Medium/Bodies/F_MED_DimeBlanketGrace/Meshes/F_MED_DimeBlanketGrace_Physics
bodies {
 base_skirt_b_l [{shapeType=Capsule, length=0.505, radius=0.505, shapeName=base_skirt_b_l_capsule, rotation=<rot 0 -90 0>}]
 base_skirt_b_m [{shapeType=Capsule, length=0.505, radius=0.505, shapeName=base_skirt_b_m_capsule, rotation=<rot 0 -90 0>}]
 base_skirt_b_r [{shapeType=Capsule, length=0.505, radius=0.505, shapeName=base_skirt_b_r_capsule, rotation=<rot 0 -90 0>}]
 base_skirt_f_l [{shapeType=Capsule, length=0.505, radius=0.505, shapeName=base_skirt_f_l_capsule, rotation=<rot 0 -90 0>}]
 base_skirt_f_m [{shapeType=Capsule, length=0.505, radius=0.505, shapeName=base_skirt_f_m_capsule, rotation=<rot 0 -90 0>}]
 base_skirt_f_r [{shapeType=Capsule, length=0.505, radius=0.505, shapeName=base_skirt_f_r_capsule, rotation=<rot 0 -90 0>}]
 dyn_skirt_b_l [{shapeType=Capsule, length=6.7255, radius=2.6244, shapeName=dyn_skirt_b_l_capsule, center=<2 -0.8675 0>}]
 dyn_skirt_b_m [{shapeType=Capsule, length=9.9986, radius=2.5513, shapeName=dyn_skirt_b_m_capsule, center=<2 0 0>}]
 dyn_skirt_b_r [{shapeType=Capsule, length=6.8684, radius=2.7297, shapeName=dyn_skirt_b_r_capsule, center=<2 -0.8675 0>}]
 dyn_skirt_f_l [{shapeType=Capsule, length=7.9472, radius=3.1159, shapeName=dyn_skirt_f_l_capsule, center=<2 -1.8927 0>}]
 dyn_skirt_f_m [{shapeType=Capsule, length=6.8293, radius=2.5306, shapeName=dyn_skirt_f_m_capsule, center=<2 0 0>}]
 dyn_skirt_f_r [{shapeType=Capsule, length=8.5504, radius=2.8545, shapeName=dyn_skirt_f_r_capsule, center=<2 -1.8927 0>}]
 pelvis [{shapeType=Capsule, length=2, radius=2, shapeName=pelvis_capsule}]
 thigh_l [{
   shapeType=Capsule
   length=18.7163
   radius=8.3347
   shapeName=thigh_r_capsule
   center=<-22.4977 -0.9857 0.2497>
   rotation=<rot 0.6206 87.0307 -88.0251>
  }]
 thigh_r [{
   shapeType=Capsule
   length=18.7163
   radius=8.3347
   shapeName=thigh_r_capsule
   center=<22.4977 0.9857 -0.2497>
   rotation=<rot 0.6206 87.0307 -88.0251>
  }]
}
constraints [
 {
  bone1Name=thigh_l
  bone2Name=pelvis
  swing1LimitDegrees=45
  swing1Motion=Free
  swing2LimitDegrees=45
  swing2Motion=Free
  twistLimitDegrees=45
  twistMotion=Free
 }
 {
  bone1Name=thigh_r
  bone2Name=pelvis
  swing1LimitDegrees=45
  swing1Motion=Free
  swing2LimitDegrees=45
  swing2Motion=Free
  twistLimitDegrees=45
  twistMotion=Free
 }
 {
  bone1Name=base_skirt_b_m
  bone2Name=pelvis
  swing1LimitDegrees=45
  swing1Motion=Free
  swing2LimitDegrees=45
  swing2Motion=Free
  twistLimitDegrees=45
  twistMotion=Free
 }
 {bone1Name=dyn_skirt_b_m, bone2Name=base_skirt_b_m, swing1LimitDegrees=10, swing2LimitDegrees=1, twistLimitDegrees=1}
 {
  bone1Name=base_skirt_f_m
  bone2Name=pelvis
  swing1LimitDegrees=45
  swing1Motion=Free
  swing2LimitDegrees=45
  swing2Motion=Free
  twistLimitDegrees=45
  twistMotion=Free
 }
 {bone1Name=dyn_skirt_f_m, bone2Name=base_skirt_f_m, swing1LimitDegrees=10, swing2LimitDegrees=1, twistLimitDegrees=1}
 {
  bone1Name=base_skirt_b_l
  bone2Name=thigh_l
  swing1LimitDegrees=45
  swing1Motion=Free
  swing2LimitDegrees=45
  swing2Motion=Free
  twistLimitDegrees=45
  twistMotion=Free
 }
 {bone1Name=dyn_skirt_b_l, bone2Name=base_skirt_b_l, swing1LimitDegrees=10, swing2LimitDegrees=1, twistLimitDegrees=1}
 {
  bone1Name=base_skirt_f_l
  bone2Name=thigh_l
  swing1LimitDegrees=45
  swing1Motion=Free
  swing2LimitDegrees=45
  swing2Motion=Free
  twistLimitDegrees=45
  twistMotion=Free
 }
 {bone1Name=dyn_skirt_f_l, bone2Name=base_skirt_f_l, swing1LimitDegrees=10, swing2LimitDegrees=1, twistLimitDegrees=1}
 {
  bone1Name=base_skirt_b_r
  bone2Name=thigh_r
  swing1LimitDegrees=45
  swing1Motion=Free
  swing2LimitDegrees=45
  swing2Motion=Free
  twistLimitDegrees=45
  twistMotion=Free
 }
 {bone1Name=dyn_skirt_b_r, bone2Name=base_skirt_b_r, swing1LimitDegrees=10, swing2LimitDegrees=1, twistLimitDegrees=1}
 {
  bone1Name=base_skirt_f_r
  bone2Name=thigh_r
  swing1LimitDegrees=45
  swing1Motion=Free
  swing2LimitDegrees=45
  swing2Motion=Free
  twistLimitDegrees=45
  twistMotion=Free
 }
 {bone1Name=dyn_skirt_f_r, bone2Name=base_skirt_f_r, swing1LimitDegrees=10, swing2LimitDegrees=1, twistLimitDegrees=1}
]
